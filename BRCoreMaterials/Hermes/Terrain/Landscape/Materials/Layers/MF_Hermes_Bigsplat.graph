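# Landscape material function for the Hermes terrain Bigsplat layer, height-blended over the incoming material under the BiomeD layer switch.
# World-scaled UVs sample diffuse, spec and normal textures; spec red drives specular and roughness, and diffuse alpha drives adjustable displacement.
# It also outputs a layer mask, mobile material and static-mesh material, using shading-path switches and derived normals on mobile.
# The colour tint and mobile BiomeD switch connect to no output.
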MaterialFunction /BRCoreMaterials/Hermes/Terrain/Landscape/Materials/Layers/MF_Hermes_Bigsplat
graph {nodes [
  n1 = NamedRerouteUsage()
  n2 = NamedRerouteUsage()
  n3 = NamedRerouteUsage()
  n4 = NamedRerouteUsage()
  n5 = NamedRerouteUsage()
  n6 = MaterialFunctionCall!(InputMaterial=@n1, LayerMaterial=@n2, LayerWeight=@n3, UseContrast=@n4, Contrast=@n5) {materialFunction=/Game/Athena/Helios/Environments/Landscape/Materials/MaterialFunctions_Utilities/MF_HeightBlendLandscapeLayers_OpacityMaskDisplaceCustom}
  FunctionOutput(None=LandscapeLayerSwitch!(LayerUsed=@n6.BlendedMaterial, LayerNotUsed=@n1) {parameterName=BiomeD}) {outputName=BlendedMaterial, sortPriority=-300}
  FunctionOutput(None=ShadingPathSwitch(Default=NamedRerouteUsage(), Mobile=NamedRerouteUsage())) {outputName=LayerMask, sortPriority=-1}
  n7 = NamedRerouteUsage()
  FunctionOutput(None=@n7) {outputName=MobileBlendedMaterial, sortPriority=-100}
  n8 = NamedRerouteUsage()
  n9 = StaticBool()
  FunctionOutput(None=ShadingPathSwitch(
    Default=NamedRerouteUsage()
    Mobile=MaterialFunctionCall!(MobileMaterial=@n8, TransformTangents=@n9) {materialFunction=/Game/Athena/Helios/Environments/Landscape/Materials/MaterialFunctions_Utilities/MF_MobileDeriveNormals}.ResultMaterial
   )) {outputName=StaticMeshMaterial, sortPriority=10}
  n10 = MaterialFunctionCall!(
   InputMaterial=@n7
   LayerMaterial=NamedRerouteUsage()
   LayerWeight=NamedRerouteUsage()
   UseContrast=NamedRerouteUsage()
   Contrast=NamedRerouteUsage()
  ) {materialFunction=/Game/Athena/Helios/Environments/Landscape/Materials/MaterialFunctions_Utilities/MF_HeightBlendLandscapeLayers_OpacityMask}
  LandscapeLayerSwitch(LayerUsed=@n10.BlendedMaterial, LayerNotUsed=@n7) {parameterName=BiomeD}
  MaterialFunctionCall(InputMaterial=@n1, LayerMaterial=@n2, LayerWeight=@n3, UseContrast=@n4, Contrast=@n5) {materialFunction=&MF_HeightBlendLandscapeLayers_OpacityMask}
  n11 = FunctionInput(Preview=StaticBool()) {
   bUsePreviewValueAsDefault=true
   description="Default = False"
   inputName=UseTangentNormals
   inputType=FunctionInput_StaticBool
   sortPriority=15
  }
  NamedRerouteDeclaration(None=@n11) {name=UseTangentNormalsBool, nodeColor=<rgba 0.5 0 1 1>}
  NamedRerouteDeclaration(None=FunctionInput(Preview=Constant()) {bUsePreviewValueAsDefault=true, inputName=VariationMasks Mobile, inputType=FunctionInput_Vector4, sortPriority=40}) {name=VariationMasks Mobile, nodeColor=<rgba 1 0.8498 0.0704 1>}
  n12 = MaterialFunctionCall!(Scale=ScalarParameter() {defaultValue=900, group=Bigsplat, parameterName=Bigsplat_WorldScale, sortPriority=-500}) {materialFunction=/Game/Athena/Apollo/Environments/Landscape/Materials/Biome/CommonMaterials/MF_SimpleWorldUV}
  n13 = TextureSampleParameter2D(UVs=@n12.Result) {
   group=Bigsplat
   parameterName=Bigsplat_Diffuse
   samplerSource=SSM_Wrap_WorldGroupSettings
   sortPriority=-400
   texture=/Hermes/Sets/BigSplat/Textures/T_BigSplat_Ground_B_D
  }
  n14 = TextureSampleParameter2D(UVs=@n12.Result) {
   group=Bigsplat
   parameterName=Bigsplat_Spec
   samplerSource=SSM_Wrap_WorldGroupSettings
   samplerType=SAMPLERTYPE_Masks
   sortPriority=-300
   texture=/Hermes/Sets/BigSplat/Textures/T_BigSplat_Ground_B_S
  }
  n15 = TextureSampleParameter2D(UVs=@n12.Result) {
   group=Bigsplat
   parameterName=Bigsplat_Normal
   samplerSource=SSM_Wrap_WorldGroupSettings
   samplerType=SAMPLERTYPE_Normal
   sortPriority=-200
   texture=/Hermes/Sets/BigSplat/Textures/T_BigSplat_Ground_B_N
  }
  n16 = Multiply(
   A=Power(Base=@n13.A, Exp=ScalarParameter() {defaultValue=1, group=Displacement, parameterName=Bigsplat_DisplacementPower})
   B=ScalarParameter() {defaultValue=1, group=Displacement, parameterName=Bigsplat_DisplacementMod}
  )
  n17 = SetMaterialAttributes!(
   MaterialAttributes
   BaseColor=@n13.RGB
   Specular=@n14.R
   Roughness=@n14.R
   Normal=MaterialFunctionCall!(Normals=@n15.RGB, TransformToTangentSpace=NamedRerouteUsage(), Normal Strength) {materialFunction=/Game/Athena/Apollo/Environments/Landscape/Materials/Biome/CommonMaterials/MF_HQ_WorldAlignedNormal}.Result
   OpacityMask=@n13.RGB
   Displacement=LinearInterpolate(B=Saturate(None=@n16), Alpha=ScalarParameter() {defaultValue=1, group=Displacement, parameterName=Bigsplat_DisplacementAmt, sortPriority=0}) {constA=0.5}
  ) {attributeSetTypes [
    #1
    #2
    #3
    #4
    #5
    #6
   ]}
  NamedRerouteDeclaration(None=@n17) {name=LayerMaterial, nodeColor=<rgba 0 1 0 1>}
  NamedRerouteDeclaration(None=FunctionInput(Preview=Constant()) {bUsePreviewValueAsDefault=true, inputName=LayerWeight, inputType=FunctionInput_Scalar, sortPriority=15}) {name=LayerWeight, nodeColor=<rgba 0.5 0 1 1>}
  NamedRerouteDeclaration(None=@n6.LayerMask) {name=LayerMask, nodeColor=<rgba 0.7344 0 1 1>}
  n18 = SetMaterialAttributes!(MaterialAttributes)
  NamedRerouteDeclaration(None=FunctionInput(Preview=@n18) {bUsePreviewValueAsDefault=true, inputName=Material, inputType=FunctionInput_MaterialAttributes}) {name=InputMaterial, nodeColor=<rgba 0 1 0 1>}
  NamedRerouteDeclaration(None=FunctionInput(Preview=Constant()) {bUsePreviewValueAsDefault=true, inputName=TerrainMasks_QuadA, inputType=FunctionInput_Vector4, sortPriority=20}) {name=TerrainMasks_QuadA, nodeColor=<rgba 0.5 0 1 1>}
  NamedRerouteDeclaration(None=FunctionInput(Preview=Constant()) {bUsePreviewValueAsDefault=true, inputName=TerrainMasks_QuadC, inputType=FunctionInput_Vector4, sortPriority=30}) {name=TerrainMasks_QuadC, nodeColor=<rgba 0.5 0 1 1>}
  NamedRerouteDeclaration(None=@n10.LayerMask) {name=MobileLayerMask, nodeColor=<rgba 0 0.9062 1 1>}
  NamedRerouteDeclaration(None=FunctionInput(Preview=Constant()) {bUsePreviewValueAsDefault=true, inputName=TerrainMasks_QuadD, inputType=FunctionInput_Vector4, sortPriority=35}) {name=TerrainMasks_QuadD, nodeColor=<rgba 0.5 0 1 1>}
  NamedRerouteDeclaration(None=FunctionInput(Preview=Constant()) {bUsePreviewValueAsDefault=true, inputName=TerrainMasks_QuadB, inputType=FunctionInput_Vector4, sortPriority=25}) {name=TerrainMasks_QuadB, nodeColor=<rgba 0.5 0 1 1>}
  NamedRerouteDeclaration(None=NamedRerouteDeclaration(None=FunctionInput(Preview=@n18) {bUsePreviewValueAsDefault=true, inputName=MobileMaterial, inputType=FunctionInput_MaterialAttributes, sortPriority=10}) {name=MobileInputMaterial, nodeColor=<rgba 0 0.5 1 1>}) {name=MobileLayerMaterial, nodeColor=<rgba 0 0.5 1 1>}
  NamedRerouteDeclaration(None=StaticBool() {value=true}) {name=UseContrast, nodeColor=<rgba 0.5 0 1 1>}
  NamedRerouteDeclaration(None=ScalarParameter() {defaultValue=0.5, group=Bigsplat, parameterName=BigsplatLayerBlendContrast, sortPriority=500}) {name=LayerBlendContrast, nodeColor=<rgba 0.5 0 1 1>}
  Saturate(None=Multiply(A=VectorParameter() {group=Bigsplat, parameterName=Bigsplat_ColorTint, defaultValue=<rgba 1 1 1 0>}.RGB, B=@n13.RGB))
 ]}
