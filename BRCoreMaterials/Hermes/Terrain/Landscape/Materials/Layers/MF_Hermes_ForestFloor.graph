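# Landscape layer function for Hermes forest-floor ground, height-blended over the incoming material under the BiomeA layer switch.
# World-UV textures drive surface attributes, world-aligned normals and displacement, while chained lerps build per-biome leaf and floor colors from terrain and texture masks.
# A separate mobile function feeds a simpler blend, but the mobile output's quality switch bypasses that blend.
# The Doomstadt leaves height-blend branch is present but not connected to the outputs.
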MaterialFunction /BRCoreMaterials/Hermes/Terrain/Landscape/Materials/Layers/MF_Hermes_ForestFloor
graph {nodes [
  n1 = Constant()
  n2 = ComponentMask(None=NamedRerouteDeclaration(None=FunctionInput(Preview=@n1) {bUsePreviewValueAsDefault=true, inputName=TerrainMasks_QuadC, inputType=FunctionInput_Vector4, sortPriority=30}) {name=TerrainMasks_QuadC, nodeColor=<rgba 0.5 0 1 1>}) {a=true}
  n3 = NamedRerouteUsage()
  Composite!(DoomstadtTrees_Mask=@n2, LayerWeight=@n3) {subgraphName=ForestFloor_High}
  Constant() {r=0.75}
  Constant() {r=1.5}
  Constant() {r=2}
  Constant() {r=900}
  n4 = NamedRerouteUsage()
  n5 = NamedRerouteUsage()
  n6 = NamedRerouteUsage()
  n7 = NamedRerouteUsage()
  n8 = NamedRerouteUsage()
  n9 = MaterialFunctionCall!(InputMaterial=@n4, LayerMaterial=@n5, LayerWeight=@n6, UseContrast=@n7, Contrast=@n8) {materialFunction=/Game/Athena/Helios/Environments/Landscape/Materials/MaterialFunctions_Utilities/MF_HeightBlendLandscapeLayers_OpacityMaskDisplaceCustom}
  FunctionOutput(None=LandscapeLayerSwitch!(LayerUsed=@n9.BlendedMaterial, LayerNotUsed=@n4) {parameterName=BiomeA}) {outputName=BlendedMaterial, sortPriority=-300}
  n10 = NamedRerouteUsage()
  FunctionOutput(None=QualitySwitch(Default=@n10, Medium=@n10)) {outputName=MobileBlendedMaterial, sortPriority=-100}
  FunctionOutput(None=ShadingPathSwitch(Default=NamedRerouteUsage(), Mobile=NamedRerouteUsage())) {outputName=LayerMask, sortPriority=-1}
  n11 = MaterialFunctionCall!(
   InputMaterial=@n10
   LayerMaterial=NamedRerouteUsage()
   LayerWeight=NamedRerouteUsage()
   UseContrast=NamedRerouteUsage()
   Contrast=NamedRerouteUsage()
  ) {materialFunction=/Game/Athena/Helios/Environments/Landscape/Materials/MaterialFunctions_Utilities/MF_HeightBlendLandscapeLayers_OpacityMask}
  LandscapeLayerSwitch(LayerUsed=@n11.BlendedMaterial, LayerNotUsed=@n10) {parameterName=BiomeA}
  n12 = Multiply(A=LandscapeLayerCoords() {mappingScale=10}, B=ScalarParameter() {defaultValue=2, group=ForestFloor, parameterName=DoomstadtLeaves_Scale})
  n13 = TextureSample(UVs=@n12) {samplerSource=SSM_Wrap_WorldGroupSettings, texture=/Game/Environments/Helios/Terrain/Textures/T_Helios_Leaves_01_D}
  n14 = VectorParameter() {
   group=ForestFloor
   parameterName=DoomstadtLeaves_Tint
   channelNames {a=Desaturation, b="", g="", r=""}
   defaultValue=<rgba 1 0.0786 0.1682 1>
  }
  n15 = Desaturation(None=@n13.RGB, Fraction=@n14.Desaturation)
  LinearInterpolate(A=@n13.RGB, B=Saturate(None=Multiply(A=@n14, B=@n15)), Alpha=@n13.RGB)
  MaterialFunctionCall(InputMaterial=@n4, LayerMaterial=@n5, LayerWeight=@n6, UseContrast=@n7, Contrast=@n8) {materialFunction=&MF_HeightBlendLandscapeLayers_OpacityMask}
  n16 = MaterialFunctionCall!(Scale=ScalarParameter() {defaultValue=900, group=ForestFloor, parameterName=ForestFloor_Scale}) {materialFunction=/Game/Athena/Apollo/Environments/Landscape/Materials/Biome/CommonMaterials/MF_SimpleWorldUV}
  n17 = TextureSampleParameter2D(UVs=@n16.Result) {
   group=ForestFloor
   parameterName=ForestFloor_S
   samplerSource=SSM_Wrap_WorldGroupSettings
   samplerType=SAMPLERTYPE_Masks
   texture=/Hermes/Terrain/Textures/T_Hermes_ForestFloor_A_S
  }
  n18 = TextureSampleParameter2D(UVs=@n16.Result) {
   group=ForestFloor
   parameterName=ForestFloor_N
   samplerSource=SSM_Wrap_WorldGroupSettings
   samplerType=SAMPLERTYPE_Normal
   texture=/Hermes/Terrain/Textures/T_Hermes_ForestFloor_A_N
  }
  n19 = TextureSampleParameter2D(UVs=@n16.Result) {
   group=ForestFloor
   parameterName=ForestFloor_D
   samplerSource=SSM_Wrap_WorldGroupSettings
   texture=/Hermes/Terrain/Textures/T_Hermes_ForestFloor_A_D
  }
  n20 = SetMaterialAttributes!(
   MaterialAttributes
   BaseColor=NamedRerouteUsage()
   Specular=@n17.R
   Roughness=@n17.R
   Normal=MaterialFunctionCall!(
    Normals=@n18.RGB
    TransformToTangentSpace=NamedRerouteUsage()
    Normal Strength=ScalarParameter() {defaultValue=1, group=ForestFloor, parameterName=Grass_BorealForest_NormalAmt}
   ) {materialFunction=/Game/Athena/Apollo/Environments/Landscape/Materials/Biome/CommonMaterials/MF_HQ_WorldAlignedNormal}.Result
   OpacityMask=@n19.A
   Displacement=LinearInterpolate(B=@n19.A, Alpha=ScalarParameter() {defaultValue=1, group=Displacement, parameterName=ForestFloor_DisplacementAmt}) {constA=0.5}
  ) {attributeSetTypes [
    #1
    #2
    #3
    #4
    #5
    #6
   ]}
  MaterialFunctionCall!(
   Material=@n20
   TerrainMasks_QuadA=NamedRerouteUsage()
   TerrainMasks_QuadB=NamedRerouteUsage()
   TerrainMasks_QuadC=NamedRerouteUsage()
   TerrainMasks_QuadD=NamedRerouteUsage()
  ) {materialFunction=/Game/Athena/Helios/Environments/Landscape/Materials/MaterialFunctions_Utilities/MF_ForestFloor_BiomeColorTint}
  Multiply(A=LandscapeLayerCoords() {mappingScale=10}, B=Constant() {r=1.5})
  n21 = VectorParameter() {group=BiomeColors, parameterName=FF_LeafColor_MaskMaker, defaultValue=<rgba 0.5543 0.2115 1.5 1>}
  n22 = NamedRerouteUsage()
  Multiply(A=@n21.RGB, B=@n22)
  n23 = VectorParameter() {group=BiomeColors, parameterName=FF_LeafColor_MagicalForest, defaultValue=<rgba 0.2016 0.1714 0.0423 1>}
  n24 = NamedRerouteUsage()
  Multiply(A=@n23.RGB, B=@n24)
  n25 = FunctionInput(Preview=StaticBool()) {
   bUsePreviewValueAsDefault=true
   description="Default = False"
   inputName=UseTangentNormals
   inputType=FunctionInput_StaticBool
   sortPriority=15
  }
  NamedRerouteDeclaration(None=@n25) {name=UseTangentNormalsBool, nodeColor=<rgba 0.5 0 1 1>}
  NamedRerouteDeclaration(None=FunctionInput(Preview=Constant()) {bUsePreviewValueAsDefault=true, inputName=VariationMasks Mobile, inputType=FunctionInput_Vector4, sortPriority=36}) {name=VariationMasks Mobile, nodeColor=<rgba 1 0.9131 0 1>}
  n26 = MaterialFunctionCall(Float4=NamedRerouteUsage()) {materialFunction=&BreakOutFloat4Components}
  NamedRerouteDeclaration(None=@n26.R) {name=BM_MagicalForest, nodeColor=<rgba 0.5 0 1 1>}
  n27 = TextureSample(UVs=@n16.Result) {samplerSource=SSM_Wrap_WorldGroupSettings, samplerType=SAMPLERTYPE_Masks, texture=/Hermes/Terrain/Textures/T_Hermes_ForestFloor_A_M}
  NamedRerouteDeclaration(None=@n27.B) {name=LeafMask02, nodeColor=<rgba 0.5 0 1 1>}
  n28 = SetMaterialAttributes!(MaterialAttributes)
  NamedRerouteDeclaration(None=FunctionInput(Preview=@n28) {bUsePreviewValueAsDefault=true, inputName=Material, inputType=FunctionInput_MaterialAttributes}) {name=InputMaterial, nodeColor=<rgba 0 1 0 1>}
  MaterialExpressionReroute_0 = Reroute(None=@n20)
  NamedRerouteDeclaration(None=@MaterialExpressionReroute_0.Output) {name=LayerMaterial, nodeColor=<rgba 0 1 0 1>}
  NamedRerouteDeclaration(None=FunctionInput(Preview=Constant()) {bUsePreviewValueAsDefault=true, inputName=LayerWeight, inputType=FunctionInput_Scalar, sortPriority=15}) {name=LayerWeight, nodeColor=<rgba 0.5 0 1 1>}
  NamedRerouteDeclaration(None=@n26.G) {name=BM_RedForest, nodeColor=<rgba 0.5 0 1 1>}
  n29 = MaterialFunctionCall(Float4=NamedRerouteUsage()) {materialFunction=&BreakOutFloat4Components}
  NamedRerouteDeclaration(None=@n29.G) {name=BM_MaskMaker, nodeColor=<rgba 0.7344 0 1 1>}
  NamedRerouteDeclaration(None=@n19.RGB) {name=Input_BaseColor, nodeColor=<rgba 0.5 0 1 1>}
  NamedRerouteDeclaration(None=@n27.G) {name=LeafMask01, nodeColor=<rgba 0.5 0 1 1>}
  NamedRerouteDeclaration(None=@n9.LayerMask) {name=LayerMask, nodeColor=<rgba 0 0.2734 1 1>}
  NamedRerouteDeclaration(None=@n27.R) {name=FloorColorMask, nodeColor=<rgba 0.5 0 1 1>}
  n30 = NamedRerouteUsage()
  n31 = VectorParameter() {group=BiomeColors, parameterName=FF_LeafColor_Mountains02, defaultValue=<rgba 0.2089 0.2552 0.0856 1>}
  n32 = NamedRerouteUsage()
  n33 = VectorParameter() {group=BiomeColors, parameterName=FF_LeafColor_Grasslands02, defaultValue=<rgba 0.2593 0.3125 0.0418 1>}
  n34 = NamedRerouteUsage()
  n35 = NamedRerouteUsage()
  n36 = LinearInterpolate(
   A=LinearInterpolate(
    A=LinearInterpolate(A=@n33.RGB, B=@n31.RGB, Alpha=@n32)
    B=VectorParameter() {group=BiomeColors, parameterName=FF_LeafColor_Wetlands02, defaultValue=<rgba 0.0951 0.3021 0.0811 1>}.RGB
    Alpha=@n34
   )
   B=VectorParameter() {group=BiomeColors, parameterName=FF_LeafColor_MossyForest02, defaultValue=<rgba 0.259 0.3229 0.0651 1>}.RGB
   Alpha=@n35
  )
  n37 = VectorParameter() {group=BiomeColors, parameterName=FF_LeafColor_BorealFall02, defaultValue=<rgba 0.8 0.158 0.03 1>}
  n38 = NamedRerouteUsage()
  n39 = NamedRerouteUsage()
  n40 = NamedRerouteUsage()
  n41 = LinearInterpolate(
   A=LinearInterpolate(
    A=LinearInterpolate(A=@n36, B=@n37.RGB, Alpha=@n38)
    B=VectorParameter() {group=BiomeColors, parameterName=FF_LeafColor_MagicalForest02, defaultValue=<rgba 0.117 0.15 0.0467 1>}.RGB
    Alpha=@n39
   )
   B=VectorParameter() {group=BiomeColors, parameterName=FF_LeafColor_MaskMaker02, defaultValue=<rgba 0.474 0.1511 0.3793 1>}.RGB
   Alpha=@n40
  )
  n42 = VectorParameter() {group=BiomeColors, parameterName=FF_LeafColor_RedForest02, defaultValue=<rgba 0.7604 0.1218 0.0307 1>}
  n43 = NamedRerouteUsage()
  n44 = NamedRerouteUsage()
  n45 = ComponentMask(None=LinearInterpolate(
    A=LinearInterpolate(A=@n41, B=@n42.RGB, Alpha=@n43)
    B=VectorParameter() {group=BiomeColors, parameterName=FF_LeafColor_Bigsplat02, defaultValue=<rgba 0.0728 0.3 0.1194 1>}.RGB
    Alpha=@n44
   )) {b=true, g=true, r=true}
  n46 = NamedRerouteUsage()
  n47 = Constant() {r=2}
  n48 = LinearInterpolate(A=@n30, B=Saturate(None=Multiply(A=@n30, B=@n45)), Alpha=Saturate(None=Multiply(A=@n46, B=@n47)))
  n49 = VectorParameter() {group=BiomeColors, parameterName=FF_LeafColor_Mountains, defaultValue=<rgba 0.2135 0.1346 0.0678 1>}
  n50 = VectorParameter() {group=BiomeColors, parameterName=FF_LeafColor_Grasslands, defaultValue=<rgba 0.1532 0.2292 0.0269 1>}
  n51 = VectorParameter() {group=BiomeColors, parameterName=FF_LeafColor_Wetlands, defaultValue=<rgba 0.0962 0.2344 0.0945 1>}
  n52 = LinearInterpolate(A=LinearInterpolate(A=@n50.RGB, B=@n49.RGB, Alpha=@n32), B=@n51.RGB, Alpha=@n34)
  n53 = VectorParameter() {group=BiomeColors, parameterName=FF_LeafColor_MossyForest, defaultValue=<rgba 0.2294 0.3698 0.0963 1>}
  n54 = VectorParameter() {group=BiomeColors, parameterName=FF_LeafColor_BorealFall, defaultValue=<rgba 0.8 0.3622 0.04 1>}
  n55 = LinearInterpolate(A=LinearInterpolate(A=LinearInterpolate(A=@n52, B=@n53.RGB, Alpha=@n35), B=@n54.RGB, Alpha=@n38), B=Saturate(None=@n23.RGB), Alpha=@n39)
  n56 = Saturate(None=@n21.RGB)
  n57 = VectorParameter() {group=BiomeColors, parameterName=FF_LeafColor_RedForest, defaultValue=<rgba 0.8194 0.0541 0.0541 1>}
  n58 = VectorParameter() {group=BiomeColors, parameterName=FF_LeafColor_Bigsplat, defaultValue=<rgba 0.04 0.2 0.1542 1>}
  n59 = LinearInterpolate(A=LinearInterpolate(A=LinearInterpolate(A=@n55, B=@n56, Alpha=@n40), B=@n57.RGB, Alpha=@n43), B=@n58.RGB, Alpha=@n44)
  n60 = Saturate(None=Multiply(A=NamedRerouteUsage(), B=ComponentMask(None=@n59) {b=true, g=true, r=true}))
  n61 = Saturate(None=Multiply(A=NamedRerouteUsage(), B=Constant() {r=2}))
  n62 = NamedRerouteUsage()
  n63 = VectorParameter() {group=BiomeColors, parameterName=FF_FloorColor_Mountains, defaultValue=<rgba 0.1107 0.1927 0.0911 1>}
  n64 = VectorParameter() {group=BiomeColors, parameterName=FF_FloorColor_Grasslands, defaultValue=<rgba 0.1924 0.3385 0.0635 1>}
  n65 = VectorParameter() {group=BiomeColors, parameterName=FF_FloorColor_Wetlands, defaultValue=<rgba 0.066 0.3333 0.102 1>}
  n66 = VectorParameter() {group=BiomeColors, parameterName=FF_FloorColor_MossyForest, defaultValue=<rgba 0.1762 0.4115 0.0493 1>}
  n67 = LinearInterpolate(A=LinearInterpolate(A=LinearInterpolate(A=@n64.RGB, B=@n63.RGB, Alpha=@n32), B=@n65.RGB, Alpha=@n34), B=@n66.RGB, Alpha=@n35)
  n68 = VectorParameter() {group=BiomeColors, parameterName=FF_FloorColor_BorealFall, defaultValue=<rgba 0.1752 0.2479 0.0331 1>}
  n69 = VectorParameter() {group=BiomeColors, parameterName=FF_FloorColor_MagicalForest, defaultValue=<rgba 0.1683 0.1812 0.063 1>}
  n70 = VectorParameter() {group=BiomeColors, parameterName=FF_FloorColor_MaskMaker, defaultValue=<rgba 0.4471 0.2196 0.6706 1>}
  n71 = LinearInterpolate(A=LinearInterpolate(A=LinearInterpolate(A=@n67, B=@n68.RGB, Alpha=@n38), B=@n69.RGB, Alpha=@n39), B=@n70.RGB, Alpha=@n40)
  n72 = VectorParameter() {group=BiomeColors, parameterName=FF_FloorColor_RedForest, defaultValue=<rgba 0.9479 0.1262 0.0691 1>}
  n73 = VectorParameter() {group=BiomeColors, parameterName=FF_FloorColor_Bigsplat, defaultValue=<rgba 0.0397 0.25 0.1505 1>}
  n74 = ComponentMask(None=LinearInterpolate(A=LinearInterpolate(A=@n71, B=@n72.RGB, Alpha=@n43), B=@n73.RGB, Alpha=@n44)) {b=true, g=true, r=true}
  n75 = Multiply(A=NamedRerouteUsage(), B=Constant() {r=2})
  NamedRerouteDeclaration(None=LinearInterpolate(A=LinearInterpolate(A=@n48, B=@n60, Alpha=@n61), B=Multiply(A=@n62, B=@n74), Alpha=Saturate(None=@n75))) {name=FinalBaseColor, nodeColor=<rgba 0.5 0 1 1>}
  NamedRerouteDeclaration(None=@n29.B) {name=BM_Bigsplat, nodeColor=<rgba 0.5 0 1 1>}
  NamedRerouteDeclaration(None=@n11.LayerMaskNoContrast) {name=MobileLayerMask, nodeColor=<rgba 0 1 0.1094 1>}
  MaterialExpressionReroute_2 = Reroute(None=@n2)
  NamedRerouteDeclaration(None=@MaterialExpressionReroute_2.DoomstadtTrees_Mask) {name=BiomeMask, nodeColor=<rgba 0.5 0 1 1>}
  NamedRerouteDeclaration(None=FunctionInput(Preview=Constant()) {bUsePreviewValueAsDefault=true, inputName=TerrainMasks_QuadA, inputType=FunctionInput_Vector4, sortPriority=20}) {name=TerrainMasks_QuadA, nodeColor=<rgba 0.5 0 1 1>}
  MaterialExpressionReroute_3 = Reroute(None=@n3)
  NamedRerouteDeclaration(None=@MaterialExpressionReroute_3.LayerWeight) {name=LayerWeight1, nodeColor=<rgba 0.5 0 1 1>}
  NamedRerouteDeclaration(None=FunctionInput(Preview=Constant()) {bUsePreviewValueAsDefault=true, inputName=TerrainMasks_QuadD, inputType=FunctionInput_Vector4, sortPriority=35}) {name=TerrainMasks_QuadD, nodeColor=<rgba 0.5 0 1 1>}
  NamedRerouteDeclaration(None=FunctionInput(Preview=Constant()) {bUsePreviewValueAsDefault=true, inputName=TerrainMasks_QuadB, inputType=FunctionInput_Vector4, sortPriority=25}) {name=TerrainMasks_QuadB, nodeColor=<rgba 0.5 0 1 1>}
  NamedRerouteDeclaration(None=FunctionInput(Preview=@n28) {bUsePreviewValueAsDefault=true, inputName=MobileMaterial, inputType=FunctionInput_MaterialAttributes, sortPriority=10}) {name=MobileInputMaterial, nodeColor=<rgba 0 0.5 1 1>}
  n76 = FunctionInput() {bUsePreviewValueAsDefault=true, inputName=MobileBorealForestMaterial, inputType=FunctionInput_MaterialAttributes, sortPriority=11}
  NamedRerouteDeclaration(None=MaterialFunctionCall!(MobileBorealForestMaterial=@n76) {materialFunction=/BRCoreMaterials/Hermes/Terrain/Landscape/Materials/Layers/QualityLevel/MF_Hermes_ForestFloor_Mobile}.Result) {name=MobileLayerMaterial, nodeColor=<rgba 0 0.5 1 1>}
  NamedRerouteDeclaration(None=StaticBool()) {name=UseContrast, nodeColor=<rgba 0.5 0 1 1>}
  n77 = MaterialFunctionCall(Float4=NamedRerouteUsage()) {materialFunction=&BreakOutFloat4Components}
  NamedRerouteDeclaration(None=@n77.R) {name=BM_Mountains, nodeColor=<rgba 0.5 0 1 1>}
  NamedRerouteDeclaration(None=ScalarParameter() {group=ForestFloor, parameterName=ForestFloor_LayerBlendContrast}) {name=LayerBlendContrast, nodeColor=<rgba 0.5 0 1 1>}
  NamedRerouteDeclaration(None=@n77.G) {name=BM_Wetlands, nodeColor=<rgba 0.5 0 1 1>}
  NamedRerouteDeclaration(None=@n77.B) {name=BM_MossyForest, nodeColor=<rgba 0.5 0 1 1>}
  NamedRerouteDeclaration(None=@n77.A) {name=BM_BorealFall, nodeColor=<rgba 0.5 0 1 1>}
  PinBase() {reroutePins [
    DoomstadtTrees_Mask {expression=&self:MaterialExpressionReroute_2}
    LayerWeight {expression=&self:MaterialExpressionReroute_3}
   ]}
  PinBase!(Output=@n20) {reroutePins [Output {expression=&self:MaterialExpressionReroute_0}]}
  Reroute(None=@n50.RGB)
  Reroute(None=@n64.RGB)
  n78 = ScalarParameter() {defaultValue=1.5, group=Displacement, parameterName=ForestFloor_DisplacementPower}
  Saturate(None=Multiply(A=Power(Base=@n19.A, Exp=@n78), B=ScalarParameter() {defaultValue=2, group=Displacement, parameterName=ForestFloor_DisplacementMod}))
  n79 = NamedRerouteUsage()
  Saturate(None=Multiply(A=@n53.RGB, B=@n79))
  n80 = NamedRerouteUsage()
  Saturate(None=Multiply(A=@n51.RGB, B=@n80))
  n81 = NamedRerouteUsage()
  Saturate(None=Multiply(A=@n54.RGB, B=@n81))
  n82 = NamedRerouteUsage()
  Saturate(None=Multiply(A=@n50.RGB, B=@n82))
  n83 = NamedRerouteUsage()
  Saturate(None=Multiply(A=@n49.RGB, B=@n83))
  n84 = NamedRerouteUsage()
  Saturate(None=Multiply(A=@n57.RGB, B=@n84))
  Saturate(None=Multiply(A=@n66.RGB, B=@n79))
  Saturate(None=Multiply(A=@n65.RGB, B=@n80))
  Saturate(None=Multiply(A=@n68.RGB, B=@n81))
  Saturate(None=Multiply(A=ScalarParameter() {defaultValue=1, group=ForestFloor, parameterName=DoomstadtLeaves_FFTransMod}, B=@n19.A))
  Saturate(None=Multiply(A=@n64.RGB, B=@n82))
  Saturate(None=Multiply(A=@n63.RGB, B=@n83))
  Saturate(None=Multiply(A=@n69.RGB, B=@n24))
  Saturate(None=Multiply(A=@n72.RGB, B=@n84))
  Saturate(None=Multiply(A=@n70.RGB, B=@n22))
  n85 = NamedRerouteUsage()
  Saturate(None=Multiply(A=@n58.RGB, B=@n85))
  Saturate(None=Multiply(A=@n73.RGB, B=@n85))
  n86 = ScalarParameter() {defaultValue=1, group=Displacement, parameterName=DoomstadtLeaves_DisplacementPower}
  Saturate(None=Multiply(A=Power(Base=@n13.A, Exp=@n86), B=ScalarParameter() {defaultValue=1, group=Displacement, parameterName=DoomstadtLeaves_DisplacementMod}))
  n87 = TextureSample(UVs=@n12) {
   samplerSource=SSM_Wrap_WorldGroupSettings
   samplerType=SAMPLERTYPE_Masks
   texture=/Game/Environments/Helios/Terrain/Textures/T_Helios_Leaves_01_S
  }
  n88 = TextureSample(UVs=@n12) {
   samplerSource=SSM_Wrap_WorldGroupSettings
   samplerType=SAMPLERTYPE_Normal
   texture=/Game/Environments/Helios/Terrain/Textures/T_Helios_Leaves_01_N
  }
  n89 = SetMaterialAttributes!(MaterialAttributes, BaseColor=@n13.RGB, Specular=@n87.R, Roughness=@n87.R, Normal=@n88.RGB, OpacityMask=@n13.RGB, Displacement=@n13.RGB) {attributeSetTypes [
    #1
    #2
    #3
    #4
    #5
    #6
   ]}
  n90 = Multiply(A=NamedRerouteUsage(), B=ScalarParameter() {defaultValue=1.25, group=ForestFloor, parameterName=DoomstadtLeaves_WorldMaskMod})
  n91 = Max(A=@n19.A, B=OneMinus(None=Saturate(None=@n90)))
  n92 = ScalarParameter() {defaultValue=1, group=ForestFloor, parameterName=DoomstadtLeaves_BlendContrast}
  StaticSwitchParameter(
   True=BlendMaterialAttributes(
    A=@n20
    B=@n89
    Alpha=MaterialFunctionCall!(Diffuse 1, Diffuse 2, Transistion Phase=@n91, Height Texture 1=@n13.A, Height Texture 2=@n19.A, Contrast=@n92) {materialFunction=/Engine/Functions/Engine_MaterialFunctions02/Texturing/HeightLerpWithTwoHeightMaps}.Alpha
   )
   False=@n20
  ) {group=ForestFloor, parameterName=UseDoomstadtLeaves, sortPriority=-500}
  TextureSample() {samplerSource=SSM_Wrap_WorldGroupSettings, texture=/Game/Environments/Helios/Terrain/Textures/T_Helios_Bor_ForestFloor_02_D}
  TextureSample() {
   samplerSource=SSM_Wrap_WorldGroupSettings
   samplerType=SAMPLERTYPE_Masks
   texture=/Game/Environments/Helios/Terrain/Textures/T_Helios_Bor_ForestFloor_02_S
  }
  TextureSample() {
   samplerSource=SSM_Wrap_WorldGroupSettings
   samplerType=SAMPLERTYPE_Normal
   texture=/Game/Environments/Helios/Terrain/Textures/T_Helios_Bor_ForestFloor_02_N
  }
 ]}
