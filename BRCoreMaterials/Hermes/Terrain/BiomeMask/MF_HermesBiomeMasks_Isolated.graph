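# Material function that turns four packed Hermes terrain mask vectors into numbered, named biome mask outputs.
# Each quad's channels are broken out, with most passed straight through and the RemoveLandscapeGrassTypes mask inverted.
# Mountains and Wetlands are isolated by multiplying them by the inverse maximum of selected other biome masks, removing overlap.
# Several outputs are unnamed spare slots, the last unconnected, and unwired nodes sketch soft versus hard transition math.
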MaterialFunction /BRCoreMaterials/Hermes/Terrain/BiomeMask/MF_HermesBiomeMasks_Isolated
graph {nodes [
  n1 = FunctionInput() {bUsePreviewValueAsDefault=true, inputName=TerrainMasks_QuadA, inputType=FunctionInput_Vector4, sortPriority=-500, previewValue=<0 0 0 0>}
  n2 = MaterialFunctionCall(Float4=NamedRerouteDeclaration(None=@n1) {name=TerrainMasks_QuadA, nodeColor=<rgba 0.5 0 1 1>}) {materialFunction=&BreakOutFloat4Components}
  n3 = FunctionInput() {bUsePreviewValueAsDefault=true, inputName=TerrainMasks_QuadB, inputType=FunctionInput_Vector4, sortPriority=-400, previewValue=<0 0 0 0>}
  n4 = MaterialFunctionCall(Float4=NamedRerouteDeclaration(None=@n3) {name=TerrainMasks_QuadB, nodeColor=<rgba 0.5 0 1 1>}) {materialFunction=&BreakOutFloat4Components}
  n5 = Max(A=@n2.A, B=@n4.R)
  n6 = OneMinus(None=Max(A=Max(A=@n2.G, B=@n2.G), B=Max(A=@n5, B=@n4.G)))
  FunctionOutput(None=NamedRerouteDeclaration(None=Multiply(A=@n2.R, B=@n6)) {name=BM_Mountains, nodeColor=<rgba 0.5 0 1 1>}) {outputName=01_Mountains, sortPriority=-1000}
  n7 = OneMinus(None=Max(A=@n2.B, B=@n4.R))
  FunctionOutput(None=NamedRerouteDeclaration(None=Multiply(A=@n2.G, B=@n7)) {name=BM_Wetlands, nodeColor=<rgba 0.5 0 1 1>}) {outputName=02_Wetlands, sortPriority=-990}
  n8 = FunctionInput() {bUsePreviewValueAsDefault=true, inputName=TerrainMasks_QuadC, inputType=FunctionInput_Vector4, sortPriority=-300, previewValue=<0 0 0 0>}
  n9 = MaterialFunctionCall(Float4=NamedRerouteDeclaration(None=@n8) {name=TerrainMasks_QuadC, nodeColor=<rgba 0.5 0 1 1>}) {materialFunction=&BreakOutFloat4Components}
  FunctionOutput(None=@n9.B) {outputName=11_Bigsplat, sortPriority=-900}
  FunctionOutput(None=@n9.A) {outputName=12_None, sortPriority=-890}
  n10 = FunctionInput() {bUsePreviewValueAsDefault=true, inputName=TerrainMasks_QuadD, inputType=FunctionInput_Vector4, sortPriority=-200, previewValue=<0 0 0 0>}
  n11 = MaterialFunctionCall(Float4=NamedRerouteDeclaration(None=@n10) {name=TerrainMasks_QuadD, nodeColor=<rgba 0.5 0 1 1>}) {materialFunction=&BreakOutFloat4Components}
  FunctionOutput(None=@n11.R) {outputName=13_None, sortPriority=-880}
  FunctionOutput(None=@n11.G) {outputName=14_None, sortPriority=-870}
  FunctionOutput(None=@n11.B) {outputName=15_None, sortPriority=-860}
  FunctionOutput(None=@n11.A) {outputName=16_None, sortPriority=-850}
  FunctionOutput() {outputName=17_None, sortPriority=-840}
  FunctionOutput(None=NamedRerouteDeclaration(None=@n2.B) {name=BM_MossyForest, nodeColor=<rgba 0.5 0 1 1>}) {outputName=03_MossyForest, sortPriority=-980}
  FunctionOutput(None=NamedRerouteDeclaration(None=@n2.A) {name=BM_BorealFall, nodeColor=<rgba 0.5 0 1 1>}) {outputName=04_BorealFall, sortPriority=-970}
  FunctionOutput(None=@n4.R) {outputName=05_Goldlands, sortPriority=-960}
  FunctionOutput(None=@n4.G) {outputName=06_RedForest, sortPriority=-950}
  FunctionOutput(None=@n4.B) {outputName=07_GoldlandsWater, sortPriority=-940}
  FunctionOutput(None=OneMinus(None=@n4.A)) {outputName=08_RemoveLandscapeGrassTypes, sortPriority=-930}
  FunctionOutput(None=@n9.R) {outputName=09_Jethro, sortPriority=-920}
  FunctionOutput(None=@n9.G) {outputName=10_MaskMaker, sortPriority=-910}
  n12 = Multiply() {desc=G}
  n13 = Multiply() {desc=B}
  Multiply(A=OneMinus(None=@n12), B=@n13) {desc=Soft transition}
  Subtract(A=@n13, B=@n12) {desc=Hard transition}
 ]}
