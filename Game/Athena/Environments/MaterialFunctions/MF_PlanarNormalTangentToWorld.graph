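# MF_PlanarNormalTangentToWorld is a material function that converts tangent-space normals to world space for planar-style mapping, without using mesh UV tangents.
# It builds a basis from the vertex world normal: X is the normalized cross of constant +Y with the normal, and Y is normal cross X.
# Transform3x3Matrix applies that basis to output World Normals.
# The basis degenerates where the normal is parallel to the Y axis.
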MaterialFunction /Game/Athena/Environments/MaterialFunctions/MF_PlanarNormalTangentToWorld
graph {nodes [
  n1 = VertexNormalWS()
  n2 = Normalize(VectorInput=CrossProduct(A=Constant3Vector() {constant=<rgba 0 1 0 0>}, B=@n1))
  FunctionOutput(None=MaterialFunctionCall(VectorToTransform=FunctionInput() {inputName=Tangent Normals}, BasisX=@n2, BasisY=CrossProduct(A=@n1, B=@n2), BasisZ=VertexNormalWS()) {materialFunction=&Transform3x3Matrix}.Transformed Vector) {outputName=World Normals}
 ]}
