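# Material function that converts two-channel triplanar tangent normals into landscape tangent space, blending them by a triplanar mask.
# The Z projection keeps its channels, the X projection swaps them, and the Y projection keeps their order.
# For X and Y, the world vertex normal's sign inverts green as one-minus on negative-facing sides, avoiding branches.
# The three variants are weighted by the mask channels and summed.
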MaterialFunction /Game/Athena/Environments/MaterialFunctions/MF_TriplanarNormalTangentToLandscapeTangent
graph {nodes [
  n1 = FunctionInput() {inputName=Tangent Normals, inputType=FunctionInput_Vector2}
  n2 = ComponentMask(None=@n1) {r=true}
  n3 = ComponentMask(None=@n1) {g=true}
  n4 = FunctionInput() {inputName=Triplanar Mask, sortPriority=2}
  n5 = Multiply(A=AppendVector(A=@n2, B=@n3), B=ComponentMask(None=@n4) {b=true})
  n6 = VertexNormalWS()
  n7 = Saturate(None=Multiply(A=Sign(None=@n6)) {constB=-1})
  n8 = Subtract(A=ComponentMask(None=@n7) {r=true}, B=@n3)
  n9 = Multiply(A=AppendVector(A=Abs(None=@n8), B=@n2), B=ComponentMask(None=@n4) {r=true})
  n10 = ComponentMask(None=@n7) {g=true}
  n11 = AppendVector(A=@n2, B=Abs(None=Subtract(A=@n10, B=@n3)))
  n12 = ComponentMask(None=@n4) {g=true}
  FunctionOutput(None=Add(A=Add(A=@n5, B=@n9), B=Multiply(A=@n11, B=@n12))) {outputName=Landscape Tangent}
 ]}
