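# Material function for terrain biome decals that converts world position into UVs for a map-space texture.
# World XY has a hard-coded center subtracted, is divided by a hard-coded extent, then shifted by 0.5 to output UVs.
# The BiomeDecal parameter from FortniteMaterialParameters is split into RG and BA masks, but neither feeds the output, so the mapping ignores the collection value.
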MaterialFunction /Hermes/Terrain/BiomeDecal/MF_BiomeDecal_TexCoord
graph {nodes [
  n1 = CollectionParameter() {collection=&FortniteMaterialParameters, parameterName=BiomeDecal}
  ComponentMask(None=@n1) {g=true, r=true}
  ComponentMask(None=@n1) {a=true, b=true}
  n2 = Subtract(A=WorldPosition().XY, B=Constant2Vector() {g=10575, r=-70604})
  n3 = Constant2Vector() {g=92935.5, r=70522.5}
  FunctionOutput(None=Add(A=Divide(A=@n2, B=@n3)) {constB=0.5}) {outputName=UVs}
 ]}
