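# Landscape layer material function that blends Hermes "Goldlands" gold veins, nuggets and riverbeds into incoming material attributes.
# A channel-selected biome decal or custom mask drives color, roughness, specular, metallic, normal and displacement blends, plus grass and layer mask outputs.
# Nugget textures use triplanar UVs, and vein normals come from a static map or four-tap render-target differences.
# Mobile paths substitute constant albedo/specular fallbacks and switch normal space for landscape versus static mesh.
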MaterialFunction /BRCoreMaterials/Hermes/Terrain/Landscape/Materials/Layers/MF_Hermes_Goldlands
graph {nodes [
  FunctionOutput(None=NamedRerouteUsage()) {outputName=GrassMeshMask, sortPriority=20}
  n1 = Multiply(A=ComponentMask(None=NamedRerouteUsage()) {g=true}, B=NamedRerouteUsage())
  n2 = NamedRerouteUsage()
  FunctionOutput(None=NamedRerouteDeclaration(None=Max(A=@n1, B=@n2)) {name=Metallic Blend, nodeColor=<rgba 0.5938 0 1 1>}) {outputName=MetallicMask, sortPriority=-1}
  FunctionOutput(None=NamedRerouteUsage()) {outputName=BlendedMaterial, sortPriority=-300}
  FunctionOutput(None=NamedRerouteUsage()) {outputName=LayerMask, sortPriority=-1}
  FunctionOutput(None=NamedRerouteUsage()) {outputName=MobileBlendedMaterial, sortPriority=-100}
  FunctionOutput(None=NamedRerouteUsage()) {outputName=StaticMeshMaterial, sortPriority=10}
  n3 = StaticComponentMaskParameter(None=NamedRerouteUsage()) {defaultR=true, group=Goldlands, parameterName=GoldLands_MaskChannel}
  n4 = ScalarParameter() {defaultValue=3, group=Goldlands, parameterName=Goldlands_GrassMask_Mult}
  NamedRerouteDeclaration(None=Saturate(None=Multiply(A=@n3, B=@n4))) {name=GrassMask, nodeColor=<rgba 1 0 0.8438 1>}
  n5 = GetMaterialAttributes(None=ShadingPathSwitch(Default=NamedRerouteUsage(), Mobile=NamedRerouteUsage())) {attributeGetTypes [
    #1
    #2
    #3
    #4
    #5
    #6
   ]}
  NamedRerouteDeclaration(None=@n5.Displacement) {name=Displacement Input, nodeColor=<rgba 0 0.5781 1 1>}
  n6 = SetMaterialAttributes!(MaterialAttributes)
  NamedRerouteDeclaration(None=FunctionInput(Preview=@n6) {bUsePreviewValueAsDefault=true, inputName=MobileMaterial, inputType=FunctionInput_MaterialAttributes, sortPriority=10}) {name=MobileInputMaterial, nodeColor=<rgba 0 0.5 1 1>}
  n7 = TextureSample(UVs=NamedRerouteUsage()) {samplerType=SAMPLERTYPE_LinearColor, texture=/Hermes/Terrain/Textures/T_Hermes_GoldVeins_A_N}
  n8 = LinearInterpolate(A=AppendVector(A=@n7.B, B=@n7.B), B=Constant2Vector() {g=0.5, r=0.5}, Alpha=NamedRerouteUsage())
  NamedRerouteDeclaration(None=ComponentMask(None=@n8) {r=true}) {name=Nugget Height, nodeColor=<rgba 1 0.75 0 1>}
  NamedRerouteDeclaration(None=ComponentMask(None=@n8) {g=true}) {name=Nugget Riverbed, nodeColor=<rgba 1 0.75 0 1>}
  n9 = NamedRerouteUsage()
  n10 = Add(A=@n9) {constB=0.5}
  n11 = Constant()
  n12 = Constant()
  n13 = NamedRerouteUsage()
  NamedRerouteDeclaration(None=StaticSwitch(True=AppendVector(A=@n10, B=@n11), False=AppendVector(A=@n13, B=@n12), Value=NamedRerouteUsage())) {name=Veins TangentNormal Mobile, nodeColor=<rgba 0.1484 1 0 1>}
  n14 = TextureSample!(UVs=NamedRerouteUsage(), Tex=NamedRerouteUsage(), Bias=Constant() {r=-1}, Apply View MipBias) {mipValueMode=TMVM_MipBias, samplerType=SAMPLERTYPE_LinearColor, texture=/Hermes/Terrain/BiomeDecal/T_BiomeDecal_Static}
  NamedRerouteDeclaration(None=AppendVector(A=@n14.B, B=@n14.B)) {name=Static Normal Map, nodeColor=<rgba 1 0.6563 0 1>}
  n15 = TextureObject() {samplerType=SAMPLERTYPE_LinearColor, texture=&T_BiomeDecal_Static}
  NamedRerouteDeclaration(None=StaticSwitch(
    True=FunctionInput(Preview=@n15) {bUsePreviewValueAsDefault=true, inputName=Texture Mask, inputType=FunctionInput_Texture2D, sortPriority=16}
    False=@n15
    Value=StaticBoolParameter() {desc=Only select one option per layer as true, group=Goldlands, parameterName=Goldlands_UseCustomMask, sortPriority=-600}
   )) {name=Texture Source, nodeColor=<rgba 0 0.4141 1 1>}
  n16 = MaterialFunctionCall() {materialFunction=/Hermes/Terrain/BiomeDecal/MF_BiomeDecal_TexCoord}
  NamedRerouteDeclaration(None=StaticSwitch(
    True=FunctionInput(Preview=@n16.UVs) {bUsePreviewValueAsDefault=true, inputName=Texture UVs, inputType=FunctionInput_Vector2, sortPriority=16}
    False=@n16.UVs
    Value=StaticBoolParameter() {desc=Only select one option per layer as true, group=Goldlands, parameterName=Goldlands_UseCustomMask, sortPriority=-600}
   )) {name=Texture UVs, nodeColor=<rgba 0.1953 1 0 1>}
  n17 = MaterialFunctionCall!(UV Scale=Constant() {r=650}, Seam Sharpness) {materialFunction=/Game/Athena/Environments/MaterialFunctions/MF_TriplanarFlat}
  NamedRerouteDeclaration(None=@n17.Triplanar Mask) {name=Triplanar Mask, nodeColor=<rgba 1 0.4688 0 1>}
  NamedRerouteDeclaration(None=@n17.UVs) {name=Triplanar UVs, nodeColor=<rgba 0.1484 0 1 1>}
  NamedRerouteDeclaration(None=@n17.Seam Mask) {name=Seam Mask, nodeColor=<rgba 0 1 0.0625 1>}
  n18 = NamedRerouteUsage()
  n19 = NamedRerouteUsage()
  n20 = NamedRerouteUsage()
  n21 = NamedRerouteUsage()
  n22 = LinearInterpolate(
   A=LinearInterpolate(A=NamedRerouteUsage(), B=NamedRerouteUsage())
   B=LinearInterpolate(A=NamedRerouteUsage(), B=LinearInterpolate(A=@n19, B=@n20, Alpha=@n21))
   Alpha=NamedRerouteUsage()
  )
  n23 = ScalarParameter() {defaultValue=2, group=Goldlands, parameterName=Goldlands_NormalEdgelIntensity}
  NamedRerouteDeclaration(None=LinearInterpolate(A=NamedRerouteUsage(), B=LinearInterpolate(A=@n18, B=@n22, Alpha=@n23), Alpha=NamedRerouteUsage())) {name=Normals Blended, nodeColor=<rgba 1 0 0.2578 1>}
  NamedRerouteDeclaration(None=@n5.Normal) {name=Normal Input, nodeColor=<rgba 0 0.5781 1 1>}
  NamedRerouteDeclaration(None=LinearInterpolate(A=NamedRerouteUsage(), Alpha=NamedRerouteUsage()) {constB=0.5}) {name=Height, nodeColor=<rgba 0.995 0.99 1 1>}
  NamedRerouteDeclaration(None=@n5.Metallic) {name=Metal Input, nodeColor=<rgba 0 0.5781 1 1>}
  NamedRerouteDeclaration(None=@n5.BaseColor) {name=BaseColor Input, nodeColor=<rgba 0.5234 1 0 1>}
  n24 = StaticComponentMaskParameter(None=NamedRerouteUsage()) {defaultR=true, group=Goldlands, parameterName=GoldLands_MaskChannel}
  n25 = Add(A=@n24, B=NamedRerouteUsage())
  NamedRerouteDeclaration(None=Saturate(None=Multiply(A=@n24, B=@n25))) {name=Veins Mask, nodeColor=<rgba 1 0 0.8438 1>}
  n26 = NamedRerouteUsage()
  n27 = NamedRerouteUsage()
  n28 = LinearInterpolate(A=@n26, B=ShadingPathSwitch(Default=@n27, Mobile=Multiply(A=@n27) {constB=1.35}), Alpha=NamedRerouteUsage())
  n29 = StaticBoolParameter() {defaultValue=true, desc=Only select one option per layer as true, group=Goldlands, parameterName=Goldlands_UseRiverBeds, sortPriority=-600}
  n30 = Multiply(A=NamedRerouteUsage(), B=NamedRerouteUsage())
  n31 = Multiply(A=@n30) {constB=1.35}
  n32 = NamedRerouteUsage()
  NamedRerouteDeclaration(None=LinearInterpolate(A=StaticSwitch(True=@n28, False=@n26, Value=@n29), B=ShadingPathSwitch(Default=@n30, Mobile=@n31), Alpha=Multiply(A=@n32, B=@n32))) {name=BaseColor Blend, nodeColor=<rgba 0 1 0.8359 1>}
  NamedRerouteDeclaration(None=@n5.Roughness) {name=Roughness Input, nodeColor=<rgba 1 0.5156 0 1>}
  n33 = ComponentMask(None=NamedRerouteUsage()) {b=true}
  n34 = Multiply(A=@n33) {constB=1.3}
  n35 = NamedRerouteUsage()
  n36 = StaticBoolParameter() {defaultValue=true, desc=Only select one option per layer as true, group=Goldlands, parameterName=Goldlands_UseRiverBeds, sortPriority=-600}
  n37 = StaticSwitch(True=Max(A=@n35, B=NamedRerouteUsage()), False=@n35, Value=@n36)
  NamedRerouteDeclaration(None=LinearInterpolate(
    A=NamedRerouteUsage()
    B=ShadingPathSwitch(Default=@n33, Mobile=@n34)
    Alpha=NamedRerouteDeclaration(None=@n37) {name=Material Mask + Riverbed, nodeColor=<rgba 1 0 0.2344 1>}
   )) {name=Roughness Blend, nodeColor=<rgba 0.1953 0 1 1>}
  n38 = ComponentMask(None=NamedRerouteUsage()) {g=true, r=true}
  n39 = Add(A=@n38) {constB=-0.5}
  n40 = MaterialFunctionCall!(Tangent Normals=DeriveNormalZ(InXY=Multiply(A=@n39) {constB=2}), Vertex Normals, Triplanar Mask=NamedRerouteUsage()) {materialFunction=/Game/Athena/Environments/MaterialFunctions/MF_TriplanarNormalTangentToWorld}
  n41 = NamedRerouteUsage()
  n42 = StaticSwitch(
   True=@n40.World Normals
   False=AppendVector(
    A=MaterialFunctionCall!(Tangent Normals=@n38, Triplanar Mask=@n41) {materialFunction=/Game/Athena/Environments/MaterialFunctions/MF_TriplanarNormalTangentToLandscapeTangent}.Landscape Tangent
    B=Constant()
   )
   Value=NamedRerouteUsage()
  )
  NamedRerouteDeclaration(None=LinearInterpolate(A=ShadingPathSwitch(Default=@n40.World Normals, Mobile=@n42), B=NamedRerouteUsage(), Alpha=NamedRerouteUsage())) {name=Nugget WorldNormal, nodeColor=<rgba 0.8281 0 1 1>}
  n43 = Multiply(A=NamedRerouteUsage()) {constB=7}
  NamedRerouteDeclaration(None=Saturate(None=Subtract(A=@n43) {constB=5})) {name=Material Mask, nodeColor=<rgba 0.3828 0 1 1>}
  n44 = Multiply(A=NamedRerouteUsage()) {constB=5}
  NamedRerouteDeclaration(None=Saturate(None=Subtract(A=@n44) {constB=2.5})) {name=Normal Mask, nodeColor=<rgba 0.3828 0 1 1>}
  n45 = NamedRerouteUsage()
  n46 = StaticSwitch(
   True=Transform(None=@n45)
   False=MaterialFunctionCall!(Tangent Normals=@n45) {materialFunction=/Game/Athena/Environments/MaterialFunctions/MF_PlanarNormalTangentToWorld}.World Normals
   Value=StaticBoolParameter() {group=Goldlands, parameterName=Goldlands_TangentUVNormals}
  )
  n47 = StaticSwitch(True=VertexNormalWS(), False=@n46, Value=NamedRerouteUsage())
  n48 = StaticSwitch(True=Constant3Vector() {constant=<rgba 0.5 0.5 1 0>}, False=NamedRerouteUsage(), Value=NamedRerouteUsage())
  n49 = NamedRerouteUsage()
  NamedRerouteDeclaration(None=ShadingPathSwitch(Default=@n46, Mobile=StaticSwitch(True=@n47, False=@n48, Value=@n49))) {name=Veins WorldNormal, nodeColor=<rgba 0 0.5078 1 1>}
  n50 = NamedRerouteUsage()
  n51 = Divide() {constB=2048}
  n52 = Constant()
  n53 = AppendVector(A=@n51, B=@n52)
  n54 = StaticComponentMaskParameter(None=TextureSample!(UVs=Add(A=@n50, B=@n53), Tex=NamedRerouteUsage(), Apply View MipBias) {samplerType=SAMPLERTYPE_LinearColor, texture=/Hermes/Terrain/BiomeDecal/RT_BiomeDecal_Tex}.RGB) {defaultR=true, group=Goldlands, parameterName=GoldLands_MaskChannel}
  n55 = Multiply(A=@n51) {constB=-1}
  n56 = AppendVector(A=@n55, B=@n52)
  n57 = StaticComponentMaskParameter(None=TextureSample(UVs=Add(A=@n50, B=@n56), Tex=NamedRerouteUsage()) {samplerType=SAMPLERTYPE_LinearColor, texture=&RT_BiomeDecal_Tex}.RGB) {defaultR=true, group=Goldlands, parameterName=GoldLands_MaskChannel}
  n58 = AppendVector(A=@n52, B=@n51)
  n59 = StaticComponentMaskParameter(None=TextureSample(UVs=Add(A=@n50, B=@n58), Tex=NamedRerouteUsage()) {samplerType=SAMPLERTYPE_LinearColor, texture=&RT_BiomeDecal_Tex}.RGB) {defaultR=true, group=Goldlands, parameterName=GoldLands_MaskChannel}
  n60 = AppendVector(A=@n52, B=@n55)
  n61 = StaticComponentMaskParameter(None=TextureSample(UVs=Add(A=@n50, B=@n60), Tex=NamedRerouteUsage()) {samplerType=SAMPLERTYPE_LinearColor, texture=&RT_BiomeDecal_Tex}.RGB) {defaultR=true, group=Goldlands, parameterName=GoldLands_MaskChannel}
  NamedRerouteDeclaration(None=AppendVector(A=Subtract(A=@n54, B=@n57), B=Subtract(A=@n59, B=@n61))) {name=Real Time NormalMap, nodeColor=<rgba 0.4062 0 1 1>}
  n62 = StaticComponentMaskParameter(None=NamedRerouteUsage()) {defaultR=true, group=Goldlands, parameterName=GoldLands_MaskChannel}
  n63 = Multiply(A=NamedRerouteUsage()) {constB=0.7}
  n64 = Saturate(None=Multiply(A=@n62, B=Add(A=@n62, B=@n63)))
  NamedRerouteDeclaration(None=LinearInterpolate(Alpha=Multiply(A=@n64, B=@n64)) {constA=0.3}) {name=BaseColor Occlusion, nodeColor=<rgba 0.3828 0 1 1>}
  n65 = TextureSample(UVs=NamedRerouteUsage()) {texture=/Hermes/Terrain/Textures/T_Hermes_GoldVeins_A_D}
  n66 = NamedRerouteDeclaration(None=Constant3Vector() {constant=<rgba 0.4219 0.2334 0.0769 1>}) {name=Yellow Fallback, nodeColor=<rgba 0 0.2734 1 1>}
  n67 = NamedRerouteUsage()
  NamedRerouteDeclaration(None=ShadingPathSwitch(Default=LinearInterpolate(A=@n65.RGB, B=@n66, Alpha=@n67), Mobile=@n66)) {name=Nugget Albedo Texture, nodeColor=<rgba 0 0.1328 1 1>}
  n68 = ComponentMask(None=NamedRerouteUsage()) {g=true}
  n69 = NamedRerouteUsage()
  n70 = Saturate(None=Multiply(A=Multiply(A=@n68, B=@n69)) {constB=5})
  NamedRerouteDeclaration(None=Multiply(A=@n70, B=@n70)) {name=Riverbed Mask, nodeColor=<rgba 1 0.0469 0 1>}
  n71 = ComponentMask(None=VertexNormalWS()) {b=true}
  n72 = Multiply(A=OneMinus(None=Abs(None=@n71))) {constB=20}
  NamedRerouteDeclaration(None=Saturate(None=Subtract(A=@n72) {constB=5.5})) {name=Slope Mask, nodeColor=<rgba 1 0.7294 0.2138 1>}
  NamedRerouteDeclaration(None=@n7.RGB) {name=Nugget Normal Texture, nodeColor=<rgba 1 0.1875 0 1>}
  n73 = Multiply(A=@n9, B=ScalarParameter() {defaultValue=1, group=Goldlands, parameterName=Goldlands_NormalRealTImeIntensity}) {constB=1.5}
  n74 = Multiply(A=Add(A=@n13) {constB=-0.5}) {constB=2}
  NamedRerouteDeclaration(None=StaticSwitch(True=DeriveNormalZ(InXY=@n73), False=DeriveNormalZ(InXY=@n74), Value=NamedRerouteUsage())) {name=Veins TangentNormal, nodeColor=<rgba 0.1484 1 0 1>}
  NamedRerouteDeclaration(None=@n14.RGBA) {name=Veins Texture, nodeColor=<rgba 0 0.8075 1 1>}
  NamedRerouteDeclaration(None=@n5.Specular) {name=Specular Input, nodeColor=<rgba 0.4531 0 1 1>}
  n75 = TextureSample(UVs=NamedRerouteUsage()) {samplerType=SAMPLERTYPE_Masks, texture=/Hermes/Terrain/Textures/T_Hermes_GoldVeins_A_S}
  n76 = NamedRerouteDeclaration(None=Constant3Vector() {constant=<rgba 0.3467 0.9647 0.3185 1>}) {name=SMR Fallback, nodeColor=<rgba 0.1953 1 0 1>}
  n77 = NamedRerouteUsage()
  NamedRerouteDeclaration(None=ShadingPathSwitch(Default=LinearInterpolate(A=@n75.RGB, B=@n76, Alpha=@n77), Mobile=@n76)) {name=Nugget Specular Texture, nodeColor=<rgba 0 1 0.1328 1>}
  n78 = Multiply(A=NamedRerouteUsage()) {constB=2}
  n79 = Subtract(A=Multiply(A=NamedRerouteUsage()) {constB=3}) {constB=2}
  NamedRerouteDeclaration(None=LinearInterpolate(A=NamedRerouteUsage(), B=Add(A=@n78) {constB=-1}, Alpha=Saturate(None=@n79)) {constB=-1}) {name=Displacement Blend, nodeColor=<rgba 0.5938 0 1 1>}
  n80 = NamedRerouteUsage()
  NamedRerouteDeclaration(None=LinearInterpolate(A=NamedRerouteUsage(), B=ComponentMask(None=@n80) {r=true}, Alpha=NamedRerouteUsage())) {name=Specular Blend, nodeColor=<rgba 1 0 0.5156 1>}
  n81 = SetMaterialAttributes!(
   MaterialAttributes
   BaseColor=NamedRerouteUsage()
   Roughness=NamedRerouteUsage()
   Normal=NamedRerouteUsage()
   Specular=NamedRerouteUsage()
   Metallic=NamedRerouteUsage()
   Displacement=NamedRerouteUsage()
  ) {attributeSetTypes [
    #1
    #3
    #4
    #2
    #6
    #5
   ]}
  NamedRerouteDeclaration(None=@n81) {name=Blended Material, nodeColor=<rgba 0 1 0.3203 1>}
  NamedRerouteDeclaration(None=FunctionInput(Preview=@n6) {bUsePreviewValueAsDefault=true, inputName=Material, inputType=FunctionInput_MaterialAttributes}) {name=InputMaterial, nodeColor=<rgba 0 1 0 1>}
  n82 = VertexNormalWS()
  n83 = Constant3Vector() {constant=<rgba 0.5 0.5 1 0>}
  n84 = NamedRerouteDeclaration(None=FunctionInput(Preview=StaticBool()) {bUsePreviewValueAsDefault=true, inputName=IsStaticMesh, inputType=FunctionInput_StaticBool, sortPriority=50}) {name=IsStaticMesh, nodeColor=<rgba 1 0.2344 0 1>}
  NamedRerouteDeclaration(None=ShadingPathSwitch(Default=@n82, Mobile=StaticSwitch(True=@n82, False=@n83, Value=@n84))) {name=WorldOrTangent Normals, nodeColor=<rgba 0 1 0.9297 1>}
  NamedRerouteDeclaration(None=StaticBoolParameter() {group=Goldlands, parameterName=Goldlands_RealTimeNormals}) {name=Real Time Normals, nodeColor=<rgba 1 0.0254 0 1>}
  ScalarParameter() {defaultValue=3, group=Goldlands, parameterName=Goldlands_NormalNuggetIntensity}
 ]}
