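# Mobile quality variant of the Hermes landscape mud layer, which adjusts an input material and outputs it through the mud biome color function.
# A squared, inverted product of OpacityMask and the variation mask's blue channel darkens base color, which is then tinted by Mobile Mud Color.
# It also inverts and squares OpacityMask and slightly lowers roughness.
# MF_Hermes_Mud_BiomeColor receives this material, the untinted darkened color and four terrain mask quads.
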MaterialFunction /BRCoreMaterials/Hermes/Terrain/Landscape/Materials/Layers/QualityLevel/MF_Hermes_Mud_Mobile
graph {nodes [
  n1 = GetMaterialAttributes(None=NamedRerouteUsage()) {attributeGetTypes [
    #1
    #2
    #3
    #4
   ]}
  n2 = ComponentMask(None=NamedRerouteUsage()) {b=true}
  n3 = Saturate(None=OneMinus(None=Multiply(A=@n1.OpacityMask, B=@n2)))
  n4 = Multiply(A=@n3, B=@n3) {constB=2}
  n5 = Multiply(
   A=@n1.BaseColor
   B=Multiply(A=@n4, B=VectorParameter() {group=Mobile Mud, parameterName=Mobile Mud Color, defaultValue=<rgba 0.7913 0.6514 0.5906 1>}.RGB)
  )
  n6 = OneMinus(None=@n1.OpacityMask)
  n7 = Multiply(A=@n6, B=@n6)
  n8 = Saturate(None=Multiply(A=@n1.Roughness) {constB=0.92})
  n9 = Constant()
  n10 = Constant()
  n11 = Constant()
  n12 = Constant()
  FunctionOutput(None=MaterialFunctionCall!(
    MobileBasecolorUnaltered=Multiply(A=@n1.BaseColor, B=@n4)
    Material=SetMaterialAttributes!(MaterialAttributes=@n1.MaterialAttributes, BaseColor=@n5, OpacityMask=@n7, Roughness=@n8) {attributeSetTypes [#1, #2, #3]}
    TerrainMasks_QuadA=FunctionInput(Preview=@n9) {bUsePreviewValueAsDefault=true, inputName=TerrainMasks_QuadA, inputType=FunctionInput_Vector4, sortPriority=-400}
    TerrainMasks_QuadB=FunctionInput(Preview=@n10) {bUsePreviewValueAsDefault=true, inputName=TerrainMasks_QuadB, inputType=FunctionInput_Vector4, sortPriority=-300}
    TerrainMasks_QuadC=FunctionInput(Preview=@n11) {bUsePreviewValueAsDefault=true, inputName=TerrainMasks_QuadC, inputType=FunctionInput_Vector4, sortPriority=-200}
    TerrainMasks_QuadD=FunctionInput(Preview=@n12) {bUsePreviewValueAsDefault=true, inputName=TerrainMasks_QuadD, inputType=FunctionInput_Vector4, sortPriority=-100}
    GrassNoise
   ) {materialFunction=/BRCoreMaterials/Hermes/Terrain/Landscape/Materials/Layers/BiomeColor/MF_Hermes_Mud_BiomeColor}.Result Mobile)
  NamedRerouteDeclaration(None=FunctionInput() {bUsePreviewValueAsDefault=true, inputName=VariationMasks Mobile, inputType=FunctionInput_Vector4, sortPriority=3}) {name=VariationMasks Mobile, nodeColor=<rgba 0.9219 1 0 1>}
  NamedRerouteDeclaration(None=FunctionInput() {bUsePreviewValueAsDefault=true, inputName=MobileInputMaterial, inputType=FunctionInput_MaterialAttributes, sortPriority=1}) {name=Input Material, nodeColor=<rgba 0 1 0.1094 1>}
 ]}
