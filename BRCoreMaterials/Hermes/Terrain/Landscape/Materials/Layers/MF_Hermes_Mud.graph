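# Mud layer function for the Hermes landscape material, outputting blended desktop and mobile materials, a layer mask and a static-mesh material.
# Derives mud from an input material by tinting and darkening gravel color, adjusting roughness and specular, flattening normals and scaling displacement, then applies biome color.
# A separate quality-level function supplies the mobile layer.
# Height-blend functions composite the layer by weight behind "Mud" landscape layer switches, with a custom-displacement variant on the non-mobile path.
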MaterialFunction /BRCoreMaterials/Hermes/Terrain/Landscape/Materials/Layers/MF_Hermes_Mud
graph {nodes [
  FunctionOutput(None=ShadingPathSwitch(Default=NamedRerouteUsage(), Mobile=NamedRerouteUsage())) {outputName=LayerMask, sortPriority=-1}
  n1 = NamedRerouteUsage()
  n2 = StaticBool()
  FunctionOutput(None=ShadingPathSwitch(
    Default=NamedRerouteUsage()
    Mobile=MaterialFunctionCall!(MobileMaterial=@n1, TransformTangents=@n2) {materialFunction=/Game/Athena/Helios/Environments/Landscape/Materials/MaterialFunctions_Utilities/MF_MobileDeriveNormals}.ResultMaterial
   )) {outputName=StaticMeshMaterial, sortPriority=10}
  n3 = NamedRerouteUsage()
  n4 = MaterialFunctionCall!(
   InputMaterial=@n3
   LayerMaterial=NamedRerouteUsage()
   LayerWeight=NamedRerouteUsage()
   UseContrast=NamedRerouteUsage()
   Contrast=NamedRerouteUsage()
  ) {materialFunction=/Game/Athena/Helios/Environments/Landscape/Materials/MaterialFunctions_Utilities/MF_HeightBlendLandscapeLayers_OpacityMask}
  FunctionOutput(None=LandscapeLayerSwitch!(LayerUsed=@n4.BlendedMaterial, LayerNotUsed=@n3) {parameterName=Mud}) {outputName=MobileBlendedMaterial, sortPriority=-100}
  n5 = NamedRerouteUsage()
  n6 = NamedRerouteUsage()
  n7 = NamedRerouteUsage()
  n8 = NamedRerouteUsage()
  n9 = NamedRerouteUsage()
  n10 = MaterialFunctionCall!(InputMaterial=@n5, LayerMaterial=@n6, LayerWeight=@n7, UseContrast=@n8, Contrast=@n9) {materialFunction=/Game/Athena/Helios/Environments/Landscape/Materials/MaterialFunctions_Utilities/MF_HeightBlendLandscapeLayers_OpacityMaskDisplaceCustom}
  FunctionOutput(None=LandscapeLayerSwitch(LayerUsed=@n10.BlendedMaterial, LayerNotUsed=@n5) {parameterName=Mud}) {outputName=BlendedMaterial, sortPriority=-300}
  MaterialFunctionCall(InputMaterial=@n5, LayerMaterial=@n6, LayerWeight=@n7, UseContrast=@n8, Contrast=@n9) {materialFunction=&MF_HeightBlendLandscapeLayers_OpacityMask}
  NamedRerouteDeclaration(None=FunctionInput(Preview=Constant()) {bUsePreviewValueAsDefault=true, inputName=VariationMasks Mobile, inputType=FunctionInput_Vector4, sortPriority=40}) {name=VariationMasks Mobile, nodeColor=<rgba 1 0.8498 0.0704 1>}
  n11 = GetMaterialAttributes(None=NamedRerouteUsage()) {attributeGetTypes [
    #1
    #2
    #3
    #4
    #5
    #6
    #7
   ]}
  n12 = SetMaterialAttributes!(
   MaterialAttributes=@n11.MaterialAttributes
   BaseColor=NamedRerouteUsage()
   OpacityMask=OneMinus(None=NamedRerouteUsage())
   Roughness=NamedRerouteUsage()
   Normal=LinearInterpolate(A=@n11.Normal, B=VertexNormalWS(), Alpha=Constant() {r=0.5})
   Specular=LinearInterpolate(A=@n11.Specular, B=Constant() {r=0.1}, Alpha=Constant() {r=0.5})
   Displacement=LinearInterpolate(B=@n11.Displacement, Alpha=ScalarParameter() {defaultValue=1, group=Displacement, parameterName=Mud_DisplacementAmt}) {constA=0.5}
  ) {attributeSetTypes [
    #1
    #5
    #3
    #4
    #2
    #6
   ]}
  NamedRerouteDeclaration(None=MaterialFunctionCall!(
    MobileBasecolorUnaltered
    Material=@n12
    TerrainMasks_QuadA=NamedRerouteUsage()
    TerrainMasks_QuadB=NamedRerouteUsage()
    TerrainMasks_QuadC=NamedRerouteUsage()
    TerrainMasks_QuadD=NamedRerouteUsage()
    GrassNoise
   ) {materialFunction=/BRCoreMaterials/Hermes/Terrain/Landscape/Materials/Layers/BiomeColor/MF_Hermes_Mud_BiomeColor}.Result) {name=New_Mud_Material, nodeColor=<rgba 0.9922 0 1 1>}
  n13 = Constant4Vector() {constant=<rgba 0.25 0.1349 0.0375 0.6>}
  n14 = Multiply(A=ComponentMask(None=Constant3Vector() {constant=<rgba 0.4 0.3052 0.2 1>}) {b=true, g=true, r=true}, B=NamedRerouteUsage())
  n15 = Saturate(None=LinearInterpolate(A=ComponentMask(None=@n13.RGBA) {b=true, g=true, r=true}, B=Saturate(None=@n14), Alpha=@n13.A))
  n16 = Multiply(A=Constant() {r=0.87}, B=@n15)
  n17 = NamedRerouteUsage()
  n18 = Multiply(A=@n17, B=@n17)
  n19 = Multiply(A=OneMinus(None=NamedRerouteUsage()), B=@n18)
  n20 = Power(Base=OneMinus(None=NamedRerouteUsage()), Exp=Constant() {r=0.15}) {constExponent=10}
  n21 = Add(A=Add(A=OneMinus(None=@n20), B=Constant() {r=0.75}), B=@n18)
  NamedRerouteDeclaration(None=Multiply(A=LinearInterpolate(A=@n15, B=@n16, Alpha=@n19), B=Saturate(None=@n21))) {name=GravelMudDiffuse, nodeColor=<rgba 0.9922 0 1 1>}
  n22 = Constant() {r=0.5}
  n23 = OneMinus(None=@n11.Roughness)
  NamedRerouteDeclaration(None=Saturate(None=Add(A=@n22, B=@n23))) {name=GravelMudRough, nodeColor=<rgba 0.9922 0 1 1>}
  NamedRerouteDeclaration(None=@n11.Opacity) {name=Gravel_Noise1, nodeColor=<rgba 0.5 0 1 1>}
  NamedRerouteDeclaration(None=@n10.LayerMask) {name=LayerMask, nodeColor=<rgba 0.1953 1 0 1>}
  NamedRerouteDeclaration(None=@n4.LayerMask) {name=MobileLayerMask, nodeColor=<rgba 0 1 0.2734 1>}
  NamedRerouteDeclaration(None=@n11.BaseColor) {name=GravDiffuse, nodeColor=<rgba 0.9922 0 1 1>}
  NamedRerouteDeclaration(None=@n11.OpacityMask) {name=GravOpacity, nodeColor=<rgba 0.9922 0 1 1>}
  n24 = SetMaterialAttributes!(MaterialAttributes)
  NamedRerouteDeclaration(None=FunctionInput(Preview=@n24) {bUsePreviewValueAsDefault=true, inputName=Material, inputType=FunctionInput_MaterialAttributes}) {name=InputMaterial, nodeColor=<rgba 0 1 0 1>}
  NamedRerouteDeclaration(None=NamedRerouteUsage()) {name=LayerMaterial, nodeColor=<rgba 0 1 0 1>}
  NamedRerouteDeclaration(None=FunctionInput(Preview=Constant()) {bUsePreviewValueAsDefault=true, inputName=LayerWeight, inputType=FunctionInput_Scalar, sortPriority=15}) {name=LayerWeight, nodeColor=<rgba 0.5 0 1 1>}
  NamedRerouteDeclaration(None=FunctionInput(Preview=Constant()) {bUsePreviewValueAsDefault=true, inputName=TerrainMasks_QuadD, inputType=FunctionInput_Vector4, sortPriority=35}) {name=TerrainMasks_QuadD, nodeColor=<rgba 0.5 0 1 1>}
  NamedRerouteDeclaration(None=FunctionInput(Preview=Constant()) {bUsePreviewValueAsDefault=true, inputName=TerrainMasks_QuadA, inputType=FunctionInput_Vector4, sortPriority=20}) {name=TerrainMasks_QuadA, nodeColor=<rgba 0.5 0 1 1>}
  NamedRerouteDeclaration(None=FunctionInput(Preview=Constant()) {bUsePreviewValueAsDefault=true, inputName=TerrainMasks_QuadC, inputType=FunctionInput_Vector4, sortPriority=30}) {name=TerrainMasks_QuadC, nodeColor=<rgba 0.5 0 1 1>}
  NamedRerouteDeclaration(None=FunctionInput(Preview=Constant()) {bUsePreviewValueAsDefault=true, inputName=TerrainMasks_QuadB, inputType=FunctionInput_Vector4, sortPriority=25}) {name=TerrainMasks_QuadB, nodeColor=<rgba 0.5 0 1 1>}
  NamedRerouteDeclaration(None=FunctionInput(Preview=@n24) {bUsePreviewValueAsDefault=true, inputName=MobileMaterial, inputType=FunctionInput_MaterialAttributes, sortPriority=10}) {name=MobileInputMaterial, nodeColor=<rgba 0 0.5 1 1>}
  NamedRerouteDeclaration(None=MaterialFunctionCall!(
    TerrainMasks_QuadA=NamedRerouteUsage()
    TerrainMasks_QuadB=NamedRerouteUsage()
    TerrainMasks_QuadC=NamedRerouteUsage()
    TerrainMasks_QuadD=NamedRerouteUsage()
    MobileInputMaterial=NamedRerouteUsage()
    VariationMasks Mobile=NamedRerouteUsage()
   ) {materialFunction=/BRCoreMaterials/Hermes/Terrain/Landscape/Materials/Layers/QualityLevel/MF_Hermes_Mud_Mobile}.Result) {name=MobileLayerMaterial, nodeColor=<rgba 0 0.5 1 1>}
  NamedRerouteDeclaration(None=StaticBool()) {name=UseContrast, nodeColor=<rgba 0.5 0 1 1>}
  n25 = FunctionInput(Preview=StaticBool()) {
   bUsePreviewValueAsDefault=true
   description="Default = False"
   inputName=UseTangentNormals
   inputType=FunctionInput_StaticBool
   sortPriority=15
  }
  NamedRerouteDeclaration(None=@n25) {name=UseTangentNormalsBool, nodeColor=<rgba 0.5 0 1 1>}
  NamedRerouteDeclaration(None=ScalarParameter() {group=Mud, parameterName=MudLayerBlendContrast}) {name=LayerBlendContrast, nodeColor=<rgba 0.5 0 1 1>}
  NamedRerouteDeclaration(None=FunctionInput() {inputName=MobileMudMaterial, inputType=FunctionInput_MaterialAttributes, sortPriority=11}) {name=MobileMudMaterial, nodeColor=<rgba 0.6172 0 1 1>}
  NamedRerouteDeclaration(None=FunctionInput(Preview=SetMaterialAttributes()) {inputName=ForMudMaterial, inputType=FunctionInput_MaterialAttributes, sortPriority=12}) {name=GravelMaterial, nodeColor=<rgba 0.5 0 1 1>}
  Reroute(None=@n12)
  ScalarParameter() {defaultValue=0.15, group=Mud, parameterName=GravelMud_HColor_Pow}
  ScalarParameter() {defaultValue=0.75, group=Mud, parameterName=GravelMud_HColor_Add}
  ScalarParameter() {defaultValue=0.87, group=Mud, parameterName=GravelMud_EdgeDarkness}
  VectorParameter() {parameterName=GravelMud_DiffuseTint, defaultValue=<rgba 0.4 0.3052 0.2 1>}
  VectorParameter() {parameterName=GravelMud_LerpColor, defaultValue=<rgba 0.25 0.1349 0.0375 0.6>}
 ]}
