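# Material function for the Hermes landscape snow layer, tied to the BiomeC layer.
# It wraps the shared Helios snow function for default shading and a Hermes-specific function for mobile, exposing blended landscape, mobile, static-mesh and post-RVT outputs.
# BiomeC layer switches gate the material-attribute blends, and the static-mesh mobile path derives normals.
# Grass and layer masks share an inverted, steepened mask chosen by shading path.
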MaterialFunction /BRCoreMaterials/Hermes/Terrain/Landscape/Materials/Layers/MF_Hermes_Snow
graph {nodes [
  n1 = NamedRerouteUsage()
  n2 = NamedRerouteUsage()
  n3 = NamedRerouteUsage()
  FunctionOutput(None=LandscapeLayerSwitch!(
    LayerUsed=BlendMaterialAttributes(A=@n1, B=@n2, Alpha=@n3) {desc="Vertex blend type altered!", vertexAttributeBlendType=UseA}
    LayerNotUsed=@n1
   ) {parameterName=BiomeC}) {outputName=MobileBlendedMaterial, sortPriority=-100}
  n4 = NamedRerouteUsage()
  n5 = StaticBool()
  FunctionOutput(None=ShadingPathSwitch(
    Default=NamedRerouteUsage()
    Mobile=MaterialFunctionCall!(MobileMaterial=@n4, TransformTangents=@n5) {materialFunction=/Game/Athena/Helios/Environments/Landscape/Materials/MaterialFunctions_Utilities/MF_MobileDeriveNormals}.ResultMaterial
   )) {outputName=StaticMeshMaterial, sortPriority=10}
  n6 = MaterialFunctionCall!(
   ScalarWeightMask=NamedRerouteUsage()
   UseSnowOcclusionTextureMask=StaticBool()
   ------
   UseMaterialInput
   AdditionalMaskToRemoveSnow
   OutputTangentNormals=NamedRerouteUsage()
   UseAdditionalMaskToRemoveSnow
   UseSnowNormals
   OuputVertexNormalsAsTheSnowNormals
   UseSparkle
   LQSpark
   AllowSnowMicroNormals
   CliffMaterialOnly=NamedRerouteUsage()
   Material=NamedRerouteUsage()
   CliffLandscapeLayerMask=NamedRerouteUsage()
   IsLandscape=StaticBool() {value=true}
   AreInputNormalsTangentSpace=StaticBool()
   UseInputNormalsForFakeSnowShadow
  ) {materialFunction=/Game/Athena/Helios/Environments/Landscape/Materials/QualityLevelLayers/MF_Helios_Snow_Default}
  FunctionOutput(None=NamedRerouteDeclaration(None=@n6.PostRVT_Landscape) {name=PostRVT, nodeColor=<rgba 0 1 0 1>}) {outputName=PostRVT, sortPriority=15}
  n7 = NamedRerouteUsage()
  n8 = NamedRerouteUsage()
  n9 = NamedRerouteUsage()
  FunctionOutput(None=LandscapeLayerSwitch(
    LayerUsed=BlendMaterialAttributes(A=@n7, B=@n8, Alpha=@n9) {desc="Vertex blend type altered!", vertexAttributeBlendType=UseA}
    LayerNotUsed=@n7
   ) {parameterName=BiomeC}) {outputName=BlendedMaterial, sortPriority=-300}
  n10 = ShadingPathSwitch(Default=NamedRerouteUsage(), Mobile=NamedRerouteUsage())
  n11 = Saturate(None=OneMinus(None=Multiply(A=@n10) {constB=5}))
  FunctionOutput(None=@n11) {outputName=GrassMask, sortPriority=5}
  FunctionOutput(None=@n11) {outputName=LayerMask, sortPriority=-1}
  NamedRerouteDeclaration(None=FunctionInput(Preview=Constant()) {bUsePreviewValueAsDefault=true, inputName=CliffLayerMask, inputType=FunctionInput_Scalar, sortPriority=50}) {name=CliffLayerMask, nodeColor=<rgba 0.5 0 1 1>}
  n12 = SetMaterialAttributes!(MaterialAttributes)
  NamedRerouteDeclaration(None=FunctionInput(Preview=@n12) {bUsePreviewValueAsDefault=true, inputName=Material, inputType=FunctionInput_MaterialAttributes}) {name=InputMaterial, nodeColor=<rgba 0 1 0 1>}
  n13 = FunctionInput(Preview=Constant()) {bUsePreviewValueAsDefault=true, inputName=LayerWeight, inputType=FunctionInput_Scalar, sortPriority=15}
  NamedRerouteDeclaration(None=@n13) {name=LayerWeight, nodeColor=<rgba 0.5 0 1 1>}
  NamedRerouteDeclaration(None=FunctionInput(Preview=SetMaterialAttributes()) {bUsePreviewValueAsDefault=true, inputName=CliffMaterialOnly, inputType=FunctionInput_MaterialAttributes, sortPriority=50}) {name=CliffMaterialOnly, nodeColor=<rgba 0.5 0 1 1>}
  n14 = FunctionInput(Preview=StaticBool()) {
   bUsePreviewValueAsDefault=true
   description="Default = False"
   inputName=UseTangentNormals
   inputType=FunctionInput_StaticBool
   sortPriority=15
  }
  NamedRerouteDeclaration(None=@n14) {name=UseTangentNormalsBool, nodeColor=<rgba 0.5 0 1 1>}
  NamedRerouteDeclaration(None=@n6.LandscapeLayerMask) {name=ModifiedLayerWeight, nodeColor=<rgba 0 1 0 1>}
  NamedRerouteDeclaration(None=FunctionInput(Preview=Constant()) {bUsePreviewValueAsDefault=true, inputName=TerrainMasks_QuadA, inputType=FunctionInput_Vector4, sortPriority=20}) {name=TerrainMasks_QuadA, nodeColor=<rgba 0.5 0 1 1>}
  NamedRerouteDeclaration(None=FunctionInput(Preview=Constant()) {bUsePreviewValueAsDefault=true, inputName=TerrainMasks_QuadC, inputType=FunctionInput_Vector4, sortPriority=30}) {name=TerrainMasks_QuadC, nodeColor=<rgba 0.5 0 1 1>}
  NamedRerouteDeclaration(None=@n6.SnowMaterialOnly) {name=LayerMaterial_StaticMesh, nodeColor=<rgba 0 1 0 1>}
  NamedRerouteDeclaration(None=FunctionInput(Preview=Constant()) {bUsePreviewValueAsDefault=true, inputName=TerrainMasks_QuadD, inputType=FunctionInput_Vector4, sortPriority=35}) {name=TerrainMasks_QuadD, nodeColor=<rgba 0.5 0 1 1>}
  NamedRerouteDeclaration(None=FunctionInput(Preview=Constant()) {bUsePreviewValueAsDefault=true, inputName=TerrainMasks_QuadB, inputType=FunctionInput_Vector4, sortPriority=25}) {name=TerrainMasks_QuadB, nodeColor=<rgba 0.5 0 1 1>}
  NamedRerouteDeclaration(None=FunctionInput(Preview=@n12) {bUsePreviewValueAsDefault=true, inputName=MobileMaterial, inputType=FunctionInput_MaterialAttributes, sortPriority=10}) {name=MobileInputMaterial, nodeColor=<rgba 0 0.5 1 1>}
  n15 = MaterialFunctionCall!(InputLayer=NamedRerouteUsage(), LayerWeight=NamedRerouteUsage(), VariationMasks Mobile=NamedRerouteUsage()) {materialFunction=/BRCoreMaterials/Hermes/Terrain/Landscape/Materials/Layers/QualityLevel/MF_Hermes_Snow_Mobile}
  NamedRerouteDeclaration(None=@n15.Result) {name=MobileLayerMaterial, nodeColor=<rgba 0 0.5 1 1>}
  NamedRerouteDeclaration(None=StaticBool()) {name=UseContrast, nodeColor=<rgba 0.5 0 1 1>}
  NamedRerouteDeclaration(None=@n6.FakeShadowSnowBasecolor) {name=FakeShadowSnowBasecolor, nodeColor=<rgba 0 1 0 1>}
  NamedRerouteDeclaration(None=ScalarParameter() {group=Snow, parameterName=SnowLayerBlendContrast}) {name=LayerBlendContrast, nodeColor=<rgba 0.5 0 1 1>}
  NamedRerouteDeclaration(None=@n6.MaterialWhenUsingInputMaterial) {name=LayerMaterial, nodeColor=<rgba 0 1 0 1>}
  NamedRerouteDeclaration(None=FunctionInput(Preview=Constant()) {inputName=VariationMasks Mobile, inputType=FunctionInput_Vector4, sortPriority=51}) {name=VariationMasks Mobile, nodeColor=<rgba 1 0.9131 0 1>}
  NamedRerouteDeclaration(None=@n15.Modified Layer Weight) {name=MobileLayerWeight, nodeColor=<rgba 0.3594 1 0 1>}
  n16 = FunctionInput(Preview=StaticBool()) {bUsePreviewValueAsDefault=true, description="Default=False", inputName=IsForStaticMesh, inputType=FunctionInput_StaticBool, sortPriority=13}
  StaticSwitch(True=Constant() {r=1}, False=@n13, Value=@n16) {defaultValue=true}
 ]}
