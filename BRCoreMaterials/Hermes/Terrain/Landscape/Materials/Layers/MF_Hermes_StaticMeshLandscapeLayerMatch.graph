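# Material function that exposes Hermes landscape layer materials for static meshes to match the terrain, with world-space normals.
# It calls the grass, gravel, mud, sand, asphalt, scree, snow and mossy forest functions at full weight, outputting their StaticMeshMaterial results.
# Shared biome and mobile variation masks feed every layer, and gravel also feeds mud.
# It also outputs post-RVT dynamic snow and the Mountains mask, and accepts a custom grass slope mask.
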MaterialFunction /BRCoreMaterials/Hermes/Terrain/Landscape/Materials/Layers/MF_Hermes_StaticMeshLandscapeLayerMatch
graph {
 properties {description=Normals are in worldspace.}
 nodes [
  n1 = MaterialFunctionCall!(
   Material
   MobileMaterial
   IsForStaticMesh
   UseTangentNormals=StaticBool()
   LayerWeight=Constant() {r=1}
   MudWeight
   TerrainMasks_QuadA=NamedRerouteUsage()
   TerrainMasks_QuadB=NamedRerouteUsage()
   TerrainMasks_QuadC=NamedRerouteUsage()
   TerrainMasks_QuadD=NamedRerouteUsage()
   VariationMasks Mobile=NamedRerouteUsage()
  ) {materialFunction=&/MF_Hermes_Gravel}
  FunctionOutput(None=MaterialFunctionCall!(
    Material
    MobileMaterial
    MobileMudMaterial=@n1.MobileMaterial
    ForMudMaterial=@n1.ForMudMaterial
    LayerWeight=Constant() {r=1}
    UseTangentNormals=StaticBool()
    TerrainMasks_QuadA=NamedRerouteUsage()
    TerrainMasks_QuadB=NamedRerouteUsage()
    TerrainMasks_QuadC=NamedRerouteUsage()
    TerrainMasks_QuadD=NamedRerouteUsage()
    VariationMasks Mobile=NamedRerouteUsage()
   ) {materialFunction=&/MF_Hermes_Mud}.StaticMeshMaterial) {outputName=MudLayer, sortPriority=30}
  n2 = StaticBool()
  n3 = FunctionInput(Preview=StaticBool()) {
   bUsePreviewValueAsDefault=true
   desc=Setting this to true will ignore Grass_EdgeOnly input
   description=Setting this to true will ignore Grass_EdgeOnly input
   inputName=Grass_UseCustomSlopeMask
   inputType=FunctionInput_StaticBool
   sortPriority=13
  }
  n4 = FunctionInput(Preview=Constant()) {bUsePreviewValueAsDefault=true, inputName=Grass_CustomSlopeMask, inputType=FunctionInput_Scalar, sortPriority=14}
  n5 = Constant() {r=1}
  n6 = NamedRerouteUsage()
  n7 = NamedRerouteUsage()
  n8 = NamedRerouteUsage()
  n9 = NamedRerouteUsage()
  n10 = NamedRerouteUsage()
  FunctionOutput(None=NamedRerouteDeclaration(None=MaterialFunctionCall!(
     Material
     MobileMaterial
     IsForStaticMesh
     UseTangentNormals=@n2
     UseCustomSlopeMask=@n3
     CustomSlopeMask=@n4
     Grass_EdgeOnly
     LayerWeight=@n5
     TerrainMasks_QuadA=@n6
     TerrainMasks_QuadB=@n7
     TerrainMasks_QuadC=@n8
     TerrainMasks_QuadD=@n9
     VariationMasks Mobile=@n10
    ) {materialFunction=&/MF_Hermes_Grass}.StaticMeshMaterial) {name=GrassLayer, nodeColor=<rgba 0.5 0 1 1>}) {outputName=GrassLayer, sortPriority=10}
  FunctionOutput(None=MaterialFunctionCall!(
    Material
    MobileMaterial
    UseTangentNormals=StaticBool()
    LayerWeight=Constant() {r=1}
    TerrainMasks_QuadA=NamedRerouteUsage()
    TerrainMasks_QuadB=NamedRerouteUsage()
    TerrainMasks_QuadC=NamedRerouteUsage()
    TerrainMasks_QuadD=NamedRerouteUsage()
    VariationMasks Mobile=NamedRerouteUsage()
   ) {materialFunction=/Hermes/Terrain/Landscape/Materials/Layers/MF_Hermes_Wetlands_MossyForest}.StaticMeshMaterial) {outputName=MossyForest, sortPriority=51}
  FunctionOutput(None=@n1.StaticMeshMaterial) {outputName=GravelLayer, sortPriority=20}
  FunctionOutput(None=MaterialFunctionCall!(
    Material
    MobileMaterial
    LayerWeight=Constant() {r=1}
    UseTangentNormals=StaticBool()
    TerrainMasks_QuadA=NamedRerouteUsage()
    TerrainMasks_QuadB=NamedRerouteUsage()
    TerrainMasks_QuadC=NamedRerouteUsage()
    TerrainMasks_QuadD=NamedRerouteUsage()
    VariationMasks Mobile=NamedRerouteUsage()
    OD_MD_Normals
    OD_MD_SedimentLayers
   ) {materialFunction=&/MF_Hermes_Sand}.StaticMeshMaterial) {outputName=SandLayer, sortPriority=40}
  FunctionOutput(None=MaterialFunctionCall!(
    Material
    MobileMaterial
    UseTangentNormals=StaticBool()
    LayerWeight=Constant() {r=1}
    TerrainMasks_QuadA=NamedRerouteUsage()
    TerrainMasks_QuadB=NamedRerouteUsage()
    TerrainMasks_QuadC=NamedRerouteUsage()
    TerrainMasks_QuadD=NamedRerouteUsage()
    VariationMasks Mobile=NamedRerouteUsage()
   ) {materialFunction=&/MF_Hermes_Asphalt}.StaticMeshMaterial) {outputName=AsphaltLayer, sortPriority=50}
  n11 = MaterialFunctionCall!(
   Material
   MobileMaterial
   IsForStaticMesh
   UseTangentNormals=StaticBool()
   LayerWeight=Constant() {r=1}
   TerrainMasks_QuadA=NamedRerouteUsage()
   TerrainMasks_QuadB=NamedRerouteUsage()
   TerrainMasks_QuadC=NamedRerouteUsage()
   TerrainMasks_QuadD=NamedRerouteUsage()
   CliffMaterialOnly
   CliffLayerMask
   VariationMasks Mobile=NamedRerouteUsage()
  ) {materialFunction=&/MF_Hermes_Snow}
  FunctionOutput(None=@n11.StaticMeshMaterial) {outputName=SnowLayer, sortPriority=300}
  FunctionOutput(None=@n11.PostRVT) {outputName=SnowDynamic, sortPriority=310}
  FunctionOutput(None=MaterialFunctionCall!(
    Material
    MobileMaterial
    UseTangentNormals=StaticBool()
    LayerWeight=Constant() {r=1}
    TerrainMasks_QuadA=NamedRerouteUsage()
    TerrainMasks_QuadB=NamedRerouteUsage()
    TerrainMasks_QuadC=NamedRerouteUsage()
    TerrainMasks_QuadD=NamedRerouteUsage()
    VariationMasks Mobile=NamedRerouteUsage()
   ) {materialFunction=&/MF_Hermes_Scree}.StaticMeshMaterial) {outputName=ScreeLayer, sortPriority=60}
  n12 = MaterialFunctionCall!(UseLandscapeCoords, "UseProjection(translucent)") {materialFunction=/BRCoreMaterials/Hermes/Terrain/BiomeMask/MF_Hermes_BiomeMasks}
  n13 = MaterialFunctionCall!(
   TerrainMasks_QuadA=NamedRerouteDeclaration(None=@n12.QuadA_RGBA) {name=TerrainMasks_QuadA, nodeColor=<rgba 0.5 0 1 1>}
   TerrainMasks_QuadB=NamedRerouteDeclaration(None=@n12.QuadB_RGBA) {name=TerrainMasks_QuadB, nodeColor=<rgba 0.5 0 1 1>}
   TerrainMasks_QuadC=NamedRerouteDeclaration(None=@n12.QuadC_RGBA) {name=TerrainMasks_QuadC, nodeColor=<rgba 0.5 0 1 1>}
   TerrainMasks_QuadD=NamedRerouteDeclaration(None=@n12.QuadD_RGBA) {name=TerrainMasks_QuadD, nodeColor=<rgba 0.5 0 1 1>}
  ) {materialFunction=/BRCoreMaterials/Hermes/Terrain/BiomeMask/MF_HermesBiomeMasks_Isolated}
  FunctionOutput(None=NamedRerouteDeclaration(None=@n13.01_Mountains) {name=BM_Mountains, nodeColor=<rgba 0.5 0 1 1>}) {description=Mountains biome mask, outputName=BM_Mountains, sortPriority=-200}
  NamedRerouteDeclaration(None=MaterialFunctionCall() {materialFunction=/BRCoreMaterials/Hermes/Terrain/Landscape/Materials/MaterialFunctions/MF_Hermes_VariationMasks_Mobile}.VariationMasks Mobile) {name=VariationMasks Mobile, nodeColor=<rgba 1 0.9141 0 1>}
  NamedRerouteDeclaration(None=@n13.02_Wetlands) {name=BM_Wetlands, nodeColor=<rgba 0.5 0 1 1>}
  NamedRerouteDeclaration(None=@n13.03_MossyForest) {name=BM_MossyForest, nodeColor=<rgba 0.5 0 1 1>}
  NamedRerouteDeclaration(None=@n13.04_BorealFall) {name=BM_BorealFall, nodeColor=<rgba 0.5 0 1 1>}
  NamedRerouteDeclaration(None=@n13.05_Goldlands) {name=BM_MagicalForest, nodeColor=<rgba 0.5 0 1 1>}
  NamedRerouteDeclaration(None=@n13.06_RedForest) {name=BM_RedForest, nodeColor=<rgba 0.5 0 1 1>}
 ]
}
